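# Control Rig for the F_MED_MajorSpeech cosmetic body that adds pose-driven corrections to its sleeves, pants, ribbons and katana.
# After the shared DeformRig function, helpers remap limb rotations relative to reference nulls into weights that offset base bones parenting the dyn_ bones.
# WearingGauntlets scales one right-sleeve correction, and right-forearm proximity to the katana shifts and rotates it.
# The spine-driven katana corrections are not wired into execution.
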ControlRigBlueprint /BRCosmetics/Characters/Player/Female/Medium/Bodies/F_MED_MajorSpeech/Meshes/F_MED_MajorSpeech_CtrlRig
graphs [
 New Function_ContainedGraph {nodes [
   n1 = RigVMFunctionEntryNode()
   RigVMFunction_MathQuaternionToEuler = &RigVMFunction_MathQuaternionToEuler(
    RotationOrder=XYZ
    Value=&RigUnit_GetRelativeTransformForItem(
     Child=(Type=Bone,Name="hand_l")
     Parent=(Type=Null,Name="lowerarm_l")
     bChildInitial=False
     bParentInitial=False
     Child.Name=@n1.Child
     Parent.Name=@n1.Parent
    ).RelativeTransform.Rotation
   )
   Remap = &RigVMFunction_MathDoubleRemap(
    TargetMaximum=1
    TargetMinimum=0
    Value=@RigVMFunction_MathQuaternionToEuler.Result.Z
    SourceMaximum=@n1.Source Maximum
    SourceMinimum=@n1.Source Minimum
   )
   > RigVMFunctionReturnNode(Delta Rotation=@Remap.Result)
  ]}
 New Function_ContainedGraph {nodes [
   n1 = RigVMFunctionEntryNode()
   RigVMFunction_MathQuaternionToEuler = &RigVMFunction_MathQuaternionToEuler(
    RotationOrder=XYZ
    Value=&RigUnit_GetRelativeTransformForItem(
     Child=(Type=Bone,Name="hand_l")
     Parent=(Type=Null,Name="lowerarm_l")
     bChildInitial=False
     bParentInitial=False
     Child.Name=@n1.Child
     Parent.Name=@n1.Parent
    ).RelativeTransform.Rotation
   )
   Remap = &RigVMFunction_MathDoubleRemap(
    TargetMaximum=1
    TargetMinimum=0
    Value=@RigVMFunction_MathQuaternionToEuler.Result.Y
    SourceMaximum=@n1.Source Maximum
    SourceMinimum=@n1.Source Minimum
   )
   > RigVMFunctionReturnNode(Delta Rotation=@Remap.Result)
  ]}
 New Function_ContainedGraph {nodes [
   n1 = RigVMFunctionEntryNode()
   RigVMFunction_MathQuaternionToEuler = &RigVMFunction_MathQuaternionToEuler(
    RotationOrder=XYZ
    Value=&RigUnit_GetRelativeTransformForItem(
     Child=(Type=Bone,Name="hand_l")
     Parent=(Type=Null,Name="lowerarm_l")
     bChildInitial=False
     bParentInitial=False
     Child.Name=@n1.Child
     Parent.Name=@n1.Parent
    ).RelativeTransform.Rotation
   )
   Remap = &RigVMFunction_MathDoubleRemap(
    TargetMaximum=1
    TargetMinimum=0
    Value=@RigVMFunction_MathQuaternionToEuler.Result.X
    SourceMaximum=@n1.Source Maximum
    SourceMinimum=@n1.Source Minimum
   )
   > RigVMFunctionReturnNode(Delta Rotation=@Remap.Result)
  ]}
 RigVMFunctionLibrary {nodes [
   Delta Rotation X = RigVMCollapseNode(ExecuteContext=())
   Delta Rotation Y = RigVMCollapseNode(ExecuteContext=())
   Delta Rotation Z = RigVMCollapseNode(ExecuteContext=())
  ]}
 RigVMModel {nodes [
   Delta Rotation X = RigVMFunctionReferenceNode(Child=pri, ExecuteContext=(), Parent=spine_05_null, Source Maximum=-5, Source Minimum=0.5)
   > &RigUnit_ModifyTransforms(
    ItemToModify=((Item=(Type=Bone,Name="base_katana_l"),Transform=(Rotation=(X=-0.1019,Y=0.1115,Z=0.1403,W=0.9785),Translation=(X=-1.6017,Y=-1.2944,Z=-2.8442),Scale3D=(X=1,Y=1,Z=1))))
    WeightMinimum=0
    Weight=@"Delta Rotation X".Delta Rotation
   )
   > Delta Rotation X_1 = RigVMFunctionReferenceNode(Child=spine_05, Parent=spine_05_null, Source Maximum=4, Source Minimum=0)
   > &RigUnit_ModifyTransforms(
    ItemToModify=((Item=(Type=Bone,Name="base_katana_l"),Transform=(Rotation=(X=-0.0608,Y=-0.1013,Z=0.0799,W=0.9898),Translation=(X=0,Y=0,Z=0),Scale3D=(X=1,Y=1,Z=1))))
    WeightMinimum=0
    Weight=@"Delta Rotation X_1".Delta Rotation
   )
   RigUnit_BeginExecution = &RigUnit_BeginExecution()
   FN_DeformRig_FNC = RigVMFunctionReferenceNode(
    DeformRig_ Curve="(deformRig_l_upperarm_up_alpha_72_#1=false,deformRig_r_upperarm_up_alpha_73_#2=false,deformRig_l_upperarm_fr_alpha_74_#3=False,deformRig_r_upperarm_fr_alpha_75_CB51B8 ...(truncated)"
    DeformRig_Bool="(upperarm_1_#4=True,elbow_3_#5=True,wrist_7_#6=True,groin_8_#7=True,glute_17_#8=True,knee_9_40C408324CA261 ...(truncated)"
    DeformRig_Debug=(upperarm_debug_18_#4=false,wrist_debug_20_#6=False,lat_debug_26_#9=False,groinGlute_debug_27_#7=False)
    DeformRig_Value="(upperarm_up_18_#10=(Rotation=(X=0.0,Y=-0.0,Z=0.0,W=1.0),Translation=(X=0.0,Y=0.0,Z=8.0),Scale3D=(X=0.70,Y=1.0,Z=1.0)),upperarm_fwd_fr_19_#11=(Rotation=(X=0.0,Y=-0.0,Z=0.0,W=1.0),Translation=(X=0.0, ...(truncated)"
    ExecuteContext=@RigUnit_BeginExecution.ExecutePin
   )
   > Delta Rotation Z_2 = RigVMFunctionReferenceNode(Child=lowerarm_r, Parent=lowerarm_r_null, Source Maximum=123, Source Minimum=22.5)
   > &RigUnit_ModifyTransforms(
    ItemToModify="((Item=(Type=Bone,Name=\"base_sleeve_bk_01_r\"),Transform=(Rotation=(X=-0.178,Y=0.0935,Z=-0.8446,W=0.4963),Translation=(X=7.5,Y=0,Z=0),Scale3D=(X=1,Y=1,Z=1))),(Item=(Type=Bone,Name=\"base_sleeve_up_01_r\"),Transform=(Rotation=(X=0.056,Y=0.0668,Z=-0.6403,W=0.76 ...(truncated)"
    WeightMinimum=0
    Weight=&RigVMFunction_MathDoubleMul(A=@"Delta Rotation Z_2".Delta Rotation, B=RigVMVariableNode() {variable=WearingGauntlets}.Value).Result
   )
   > Delta Rotation Y_4 = RigVMFunctionReferenceNode(Child=lowerarm_r, Parent=lowerarm_r_null, Source Maximum=90, Source Minimum=-50)
   > &RigUnit_ModifyTransforms(
    ItemToModify="((Item=(Type=Bone,Name=\"base_sleeve_bk_01_r\"),Transform=(Rotation=(X=0,Y=0,Z=0,W=1),Translation=(X=0,Y=0,Z=0),Scale3D=(X=1,Y=1,Z=1))),(Item=(Type=Bone,Name=\"base_sleeve_up_01_r\"),Transform=(Rotation=(X=0,Y=0,Z=0,W=1),Translation=(X=0,Y=0,Z=0),Scale3D=(X=1, ...(truncated)"
    WeightMinimum=0
    Weight=@"Delta Rotation Y_4".Delta Rotation
   )
   > Delta Rotation Y = RigVMFunctionReferenceNode(Child=upperarm_r, Parent=upperarm_r_null, Source Maximum=80, Source Minimum=0)
   > &RigUnit_ModifyTransforms(
    ItemToModify="((Item=(Type=Bone,Name=\"base_sleeve_dw_01_r\"),Transform=(Rotation=(X=-0.2275,Y=0.4609,Z=0.1235,W=0.8489),Translation=(X=0,Y=16,Z=0),Scale3D=(X=1,Y=1,Z=1))),(Item=(Type=Bone,Name=\"base_sleeve_fr_01_r\"),Transform=(Rotation=(X=0,Y=0.2588,Z=0,W=0.9659),Transla ...(truncated)"
    WeightMinimum=0
    Weight=@"Delta Rotation Y".Delta Rotation
   )
   > Delta Rotation Z = RigVMFunctionReferenceNode(Child=calf_r, Parent=calf_r_null, Source Maximum=90, Source Minimum=-2)
   > &RigUnit_ModifyTransforms(
    ItemToModify="((Item=(Type=Bone,Name=\"base_pants_bk_01_r\"),Transform=(Rotation=(X=-0.0006,Y=0.0087,Z=0.0698,W=0.9975),Translation=(X=0,Y=8.1,Z=0),Scale3D=(X=1,Y=1,Z=1))),(Item=(Type=Bone,Name=\"base_pants_in_01_r\"),Transform=(Rotation=(X=0.0016,Y=-0.0523,Z=0.0305,W=0.998 ...(truncated)"
    WeightMinimum=0
    Weight=@"Delta Rotation Z".Delta Rotation
   )
   > Delta Rotation Z_1 = RigVMFunctionReferenceNode(Child=calf_l, Parent=calf_l_null, Source Maximum=70, Source Minimum=-2)
   > &RigUnit_ModifyTransforms(
    ItemToModify="((Item=(Type=Bone,Name=\"base_pants_bk_01_l\"),Transform=(Rotation=(X=0.0005,Y=-0.0087,Z=0.061,W=0.9981),Translation=(X=0,Y=7,Z=0),Scale3D=(X=1,Y=1,Z=1))),(Item=(Type=Bone,Name=\"base_pants_in_01_l\"),Transform=(Rotation=(X=-0.0016,Y=0.061,Z=0.0261,W=0.9978),T ...(truncated)"
    WeightMinimum=0
    Weight=@"Delta Rotation Z_1".Delta Rotation
   )
   > Delta Rotation Z_6 = RigVMFunctionReferenceNode(Child=thigh_r, Parent=thigh_r_null, Source Maximum=35, Source Minimum=0)
   > &RigUnit_ModifyTransforms(
    ItemToModify="((Item=(Type=Bone,Name=\"base_pants_in_01_r\"),Transform=(Rotation=(X=0.0057,Y=-0.0432,Z=0.1304,W=0.9905),Translation=(X=0,Y=0,Z=0),Scale3D=(X=1,Y=1,Z=1))),(Item=(Type=Bone,Name=\"base_pants_bk_01_r\"),Transform=(Rotation=(X=0,Y=0,Z=0.1305,W=0.9914),Translatio ...(truncated)"
    WeightMinimum=0
    Weight=@"Delta Rotation Z_6".Delta Rotation
   )
   > Delta Rotation Z_7 = RigVMFunctionReferenceNode(Child=thigh_l, Parent=thigh_l_null, Source Maximum=30, Source Minimum=-3)
   > &RigUnit_ModifyTransforms(
    ItemToModify="((Item=(Type=Bone,Name=\"base_pants_in_01_l\"),Transform=(Rotation=(X=-0.008,Y=0.0605,Z=0.1303,W=0.9896),Translation=(X=0,Y=0,Z=0),Scale3D=(X=1,Y=1,Z=1))),(Item=(Type=Bone,Name=\"base_pants_bk_01_l\"),Transform=(Rotation=(X=0,Y=0,Z=0.1305,W=0.9914),Translation ...(truncated)"
    WeightMinimum=0
    Weight=@"Delta Rotation Z_7".Delta Rotation
   )
   > Delta Rotation Y_1 = RigVMFunctionReferenceNode(Child=thigh_r, Parent=thigh_r_null, Source Maximum=4)
   > &RigUnit_ModifyTransforms(
    ItemToModify="((Item=(Type=Bone,Name=\"base_pants_in_01_r\"),Transform=(Rotation=(X=0,Y=0.0436,Z=0,W=0.999),Translation=(X=0,Y=0,Z=0),Scale3D=(X=1,Y=1,Z=1))),(Item=(Type=Bone,Name=\"base_pants_bk_01_r\"),Transform=(Rotation=(X=-0.0038,Y=0.0871,Z=0.0435,W=0.9952),Translation ...(truncated)"
    WeightMinimum=0
    Weight=@"Delta Rotation Y_1".Delta Rotation
   )
   > Delta Rotation Y_2 = RigVMFunctionReferenceNode(Child=thigh_l, Parent=thigh_l_null, Source Maximum=4)
   > &RigUnit_ModifyTransforms(
    ItemToModify="((Item=(Type=Bone,Name=\"base_pants_in_01_l\"),Transform=(Rotation=(X=0,Y=-0.0436,Z=0,W=0.999),Translation=(X=0,Y=0,Z=0),Scale3D=(X=1,Y=1,Z=1))),(Item=(Type=Bone,Name=\"base_pants_bk_01_l\"),Transform=(Rotation=(X=0.0073,Y=-0.1043,Z=0.0694,W=0.9921),Translatio ...(truncated)"
    WeightMinimum=0
    Weight=@"Delta Rotation Y_2".Delta Rotation
   )
   > Delta Rotation Y_3 = RigVMFunctionReferenceNode(Child=thigh_l, Parent=thigh_l_null, Source Maximum=-50, Source Minimum=0)
   > &RigUnit_ModifyTransforms(
    ItemToModify=((Item=(Type=Bone,Name="base_katana_l"),Transform=(Rotation=(X=0,Y=-0.1521,Z=0,W=0.9884),Translation=(X=0,Y=0,Z=0),Scale3D=(X=1,Y=1,Z=1))))
    WeightMinimum=0
    Weight=@"Delta Rotation Y_3".Delta Rotation
   )
   > Delta Rotation Z_5 = RigVMFunctionReferenceNode(Child=thigh_l, Parent=thigh_l_null, Source Maximum=-25)
   > &RigUnit_ModifyTransforms(
    ItemToModify=((Item=(Type=Bone,Name="base_katana_l"),Transform=(Rotation=(X=0,Y=-0.0218,Z=0,W=0.9998),Translation=(X=0,Y=0,Z=-0.5),Scale3D=(X=1,Y=1,Z=1))))
    WeightMinimum=0
    Weight=@"Delta Rotation Z_5".Delta Rotation
   )
   > Delta Rotation Z_3 = RigVMFunctionReferenceNode(Child=thigh_l, Parent=thigh_l_null, Source Maximum=-110, Source Minimum=10)
   > &RigUnit_ModifyTransforms(
    ItemToModify=((Item=(Type=Bone,Name="base_ribbon_l"),Transform=(Rotation=(X=0.7255,Y=-0.124,Z=-0.3049,W=0.6045),Translation=(X=0,Y=0,Z=0),Scale3D=(X=1,Y=1,Z=1))))
    WeightMinimum=0
    Weight=@"Delta Rotation Z_3".Delta Rotation
   )
   > Delta Rotation Z_4 = RigVMFunctionReferenceNode(Child=thigh_r, Parent=thigh_r_null, Source Maximum=-110, Source Minimum=25)
   > &RigUnit_ModifyTransforms(
    ItemToModify=((Item=(Type=Bone,Name="base_ribbon_r"),Transform=(Rotation=(X=-0.6625,Y=0.0763,Z=0.2376,W=0.7062),Translation=(X=0,Y=0,Z=0),Scale3D=(X=1,Y=1,Z=1))))
    WeightMinimum=0
    Weight=@"Delta Rotation Z_4".Delta Rotation
   )
   RigVMFunction_MathVectorDistance = &RigVMFunction_MathVectorDistance(
    A=&RigUnit_GetTransform(Item=(Type=Bone,Name="lowerarm_r"), bInitial=False).Transform.Translation
    B=&RigUnit_GetTransform(Item=(Type=Bone,Name="dyn_katana_l"), bInitial=False).Transform.Translation
   )
   Remap = /Script/RigVM.RigVMFunction_MathVectorRemap(
    SourceMaximum=(X=23,Y=23,Z=23)
    SourceMinimum=(X=34,Y=34,Z=34)
    TargetMaximum=(X=-2.5,Y=-2.2,Z=-4)
    TargetMinimum=(X=0,Y=0,Z=0)
    Value=(X=0,Y=0,Z=0)
    bClamp=true
    Value.X=@RigVMFunction_MathVectorDistance.Result
    Value.Y=@RigVMFunction_MathVectorDistance.Result
    Value.Z=@RigVMFunction_MathVectorDistance.Result
   ) {
    is=struct RigVMUnitNode
    template="Remap::Execute(in Value,in SourceMinimum,in SourceMaximum,in TargetMinimum,in TargetMaximum,in bClamp,out Result)"
   }
   n1 = &RigVMFunction_MathVectorRemap(
    SourceMaximum=(X=23,Y=23,Z=23)
    SourceMinimum=(X=34.5,Y=34.5,Z=34.5)
    TargetMaximum=(X=-36,Y=8,Z=-18)
    TargetMinimum=(X=0,Y=0,Z=0)
    Value=(X=0,Y=0,Z=0)
    bClamp=true
    Value.X=@RigVMFunction_MathVectorDistance.Result
    Value.Y=@RigVMFunction_MathVectorDistance.Result
    Value.Z=@RigVMFunction_MathVectorDistance.Result
   ) {template="Remap::Execute(in Value,in SourceMinimum,in SourceMaximum,in TargetMinimum,in TargetMaximum,in bClamp,out Result)"}
   RigVMFunction_MathQuaternionFromEuler = &RigVMFunction_MathQuaternionFromEuler(RotationOrder=YZX, Euler=@n1.Result)
   > &RigUnit_ModifyTransforms(
    ItemToModify=((Item=(Type=Bone,Name="base_katana_l"),Transform=(Rotation=(X=-0.2168,Y=0.0114,Z=0.0293,W=0.9757),Translation=(X=-1.5,Y=0,Z=-3),Scale3D=(X=1,Y=1,Z=1))))
    Weight=1
    WeightMinimum=0
    ItemToModify.0.Transform.Translation=@Remap.Result
    ItemToModify.0.Transform.Rotation=@RigVMFunction_MathQuaternionFromEuler.Result
   )
   RigVMDispatchNode(BoneName=root, Default=1, Name=deformRig_r_upperarm_bk_alpha) {template="DISPATCH_RigDispatch_GetAnimAttribute(in Name,in BoneName,in Default,out Value,out Found)"}
   RigVMDispatchNode(Enabled=True, ExecuteContext=(), ScreenColor=(R=1,G=1,B=1,A=1), ScreenDuration=0.05, Value=(X=0,Y=0,Z=0,W=1)) {template="DISPATCH_RigVMDispatch_Print(in Prefix,in Value,in Enabled,in ScreenDuration,in ScreenColor)"}
   RigVMDispatchNode(Enabled=True, ExecuteContext=(), ScreenColor=(R=1,G=1,B=1,A=1), ScreenDuration=0.05, Value=(X=0,Y=0,Z=0)) {template="DISPATCH_RigVMDispatch_Print(in Prefix,in Value,in Enabled,in ScreenDuration,in ScreenColor)"}
   RigVMDispatchNode(Enabled=True, ExecuteContext=(), ScreenColor=(R=1,G=1,B=1,A=1), ScreenDuration=0.05, Value=0) {template="DISPATCH_RigVMDispatch_Print(in Prefix,in Value,in Enabled,in ScreenDuration,in ScreenColor)"}
   RigVMDispatchNode(Enabled=True, ExecuteContext=(), ScreenColor=(R=1,G=1,B=1,A=1), ScreenDuration=0.05, Value=0) {template="DISPATCH_RigVMDispatch_Print(in Prefix,in Value,in Enabled,in ScreenDuration,in ScreenColor)"}
  ]}
]
hierarchy [
 root {loc=<0 0 0>, rot=<0 0 0 1>}
 pelvis {parent=root, loc=<0 -0.5301 98.0571>, rot=<0.0127 0.707 -0.0127 -0.707>}
 spine_01 {parent=pelvis, loc=<6.1208 -0.9328 0>, rot=<0 0 0.034 -0.9994>}
 spine_02 {parent=spine_01, loc=<6.5997 -0.2095 0>, rot=<0 0 0.0238 -0.9997>}
 spine_03 {parent=spine_02, loc=<6.4914 -0.1237 0>, rot=<0 0 0.0149 -0.9999>}
 spine_04 {parent=spine_03, loc=<6.9598 0.0794 0>, rot=<0 0 -0.0132 -0.9999>}
 spine_05 {parent=spine_04, loc=<6.9987 0 0>, rot=<0 0 -0.0015 -1>}
 neck_01 {parent=spine_05, loc=<11.3138 0.1529 0>, rot=<0 0 -0.2003 -0.9797>}
 neck_02 {parent=neck_01, loc=<4.5627 0 0>, rot=<0 0 0.0504 -0.9987>}
 head {parent=neck_02, loc=<5.3914 0 0>, rot=<0 0 0.115 -0.9934>}
 clavicle_l {parent=spine_05, loc=<5.4831 -0.1053 -3.2946>, rot=<0.0465 -0.7289 -0.009 -0.6829>}
 upperarm_l {parent=clavicle_l, loc=<9.8658 -1.2136 -0.4105>, rot=<0.0164 -0.3245 -0.0269 -0.9454>}
 lowerarm_l {parent=upperarm_l, loc=<26.465 -0.8345 0.259>, rot=<0.0036 0.0161 -0.2298 -0.9731>}
 hand_l {parent=lowerarm_l, loc=<22.7355 -0.3203 -0.6105>, rot=<-0.6784 -0.0223 0.0735 -0.7307>}
 thumb_01_l {parent=hand_l, loc=<1.3118 -0.3191 -2.459>, rot=<0.6546 -0.4042 0.0664 -0.6354>}
 thumb_02_l {parent=thumb_01_l, loc=<5.9336 -0.0448 0.0401>, rot=<-0.0037 0.0213 0.298 -0.9543>}
 thumb_03_l {parent=thumb_02_l, loc=<2.2223 0.3682 -0.0907>, rot=<0.0118 -0.013 0.0066 -0.9998>}
 index_metacarpal_l {parent=hand_l, loc=<1.6535 0.0183 -2.96>, rot=<0.0158 -0.0124 -0.0173 -0.9996>}
 index_01_l {parent=index_metacarpal_l, loc=<8.3008 0.0357 -0.0524>, rot=<-0.0044 0.0485 0.0924 -0.9945>}
 index_02_l {parent=index_01_l, loc=<3.3587 -0.3004 0.0452>, rot=<-0.002 0.0027 0.3225 -0.9466>}
 index_03_l {parent=index_02_l, loc=<2.5578 -0.123 -0.0252>, rot=<-0.002 0.0067 0.2558 -0.9667>}
 middle_metacarpal_l {parent=hand_l, loc=<2.1856 -0.1725 -0.5984>, rot=<0.006 0.0168 -0.0329 -0.9993>}
 middle_01_l {parent=middle_metacarpal_l, loc=<7.3275 -0.2447 0.0129>, rot=<-0.0208 0.0042 0.0858 -0.9961>}
 middle_02_l {parent=middle_01_l, loc=<4.0123 0.1401 0.1175>, rot=<-0.0218 0.0258 0.3512 -0.9357>}
 middle_03_l {parent=middle_02_l, loc=<2.6077 0.0534 0.1373>, rot=<0.0082 0.0188 0.1708 -0.9851>}
 ring_metacarpal_l {parent=hand_l, loc=<2.2074 -0.3625 1.3638>, rot=<-0.0284 0.0722 -0.0458 -0.9959>}
 ring_01_l {parent=ring_metacarpal_l, loc=<7.0295 -0.6197 -0.1446>, rot=<-0.0489 -0.0252 0.1892 -0.9804>}
 ring_02_l {parent=ring_01_l, loc=<3.6174 0.4647 -0.1027>, rot=<0 0 0.2806 -0.9598>}
 ring_03_l {parent=ring_02_l, loc=<2.4152 0.114 0.0199>, rot=<0 0 0.1189 -0.9929>}
 pinky_metacarpal_l {parent=hand_l, loc=<1.7616 -0.9753 3.3814>, rot=<-0.0397 0.0814 -0.069 -0.9935>}
 pinky_01_l {parent=pinky_metacarpal_l, loc=<7.1709 -0.4831 -0.1094>, rot=<-0.0654 -0.0035 0.2242 -0.9723>}
 pinky_02_l {parent=pinky_01_l, loc=<2.8274 0.0893 -0.1043>, rot=<0.0037 0.0011 0.2854 -0.9584>}
 pinky_03_l {parent=pinky_02_l, loc=<2.2046 0.0307 -0.0078>, rot=<-0.0002 0.0098 0.1445 -0.9895>}
 deform_wrist_dn_base_l {parent=hand_l, loc=<0 -1.5326 0>, rot=<0 0 0 -1>}
 deform_wrist_dn_l {parent=deform_wrist_dn_base_l, loc=<0 0 0>, rot=<0 0 0 1>}
 deform_wrist_up_base_l {parent=hand_l, loc=<0 2.3009 0>, rot=<0 0 0 -1>}
 deform_wrist_up_l {parent=deform_wrist_up_base_l, loc=<0 0 0>, rot=<0 0 0 1>}
 lowerarm_twist_02_l {parent=lowerarm_l, loc=<10.6707 0 0>, rot=<0 0 0 1>}
 lowerarm_twist_01_l {parent=lowerarm_l, loc=<18.316 0 0>, rot=<0 0 0 1>}
 deform_elbow_in_base_l {parent=lowerarm_l, loc=<0.1834 3.6062 -0.2572>, rot=<0 0 0 -1>}
 deform_elbow_in_l {parent=deform_elbow_in_base_l, loc=<0 0 0>, rot=<0 0 0 1>}
 deform_elbow_out_base_l {parent=lowerarm_l, loc=<0 -4 0>, rot=<0 0 0 1>}
 deform_elbow_out_l {parent=deform_elbow_out_base_l, loc=<0 0 0>, rot=<0 0 0 1>}
 upperarm_twist_01_l {parent=upperarm_l, loc=<7.2135 -0.1134 0>, rot=<0 0 0 1>}
 deform_upperarm_bk_base_l {parent=upperarm_twist_01_l, loc=<-6.4053 -4.3746 0.314>, rot=<0 0 0 -1>}
 deform_upperarm_bk_l {parent=deform_upperarm_bk_base_l, loc=<0 0 0>, rot=<0 0 0 1>}
 deform_upperarm_fr_base_l {parent=upperarm_twist_01_l, loc=<-5.6312 5.092 0.3354>, rot=<0 0 0 -1>}
 deform_upperarm_fr_l {parent=deform_upperarm_fr_base_l, loc=<0 0 0>, rot=<0 0 0 1>}
 deform_upperarm_up_base_l {parent=upperarm_twist_01_l, loc=<-8.891 0.4825 3.9517>, rot=<0 0 0 -1>}
 deform_upperarm_up_l {parent=deform_upperarm_up_base_l, loc=<0 0 0>, rot=<0 0 0 1>}
 upperarm_twist_02_l {parent=upperarm_l, loc=<19.7998 -0.1134 0>, rot=<0 0 0 1>}
 base_sleeve_dw_01_l {parent=upperarm_l, loc=<13.1062 0.9131 -6.6362>, rot=<0 -0.9962 0 0.0872>}
 dyn_sleeve_dw_01_l {parent=base_sleeve_dw_01_l, loc=<0 0 0>, rot=<0 0 0 1>}
 base_sleeve_up_01_l {parent=upperarm_l, loc=<14.671 -0.7791 3.9855>, rot=<0 -0.999 0 -0.0436>}
 dyn_sleeve_up_01_l {parent=base_sleeve_up_01_l, loc=<0 0 0>, rot=<0 0 0 1>}
 base_sleeve_fr_01_l1 {parent=upperarm_l, loc=<12.9818 4.9278 -3.1362>, rot=<0 -0.999 0 0.0436>}
 dyn_sleeve_fr_01_l {parent=base_sleeve_fr_01_l1, loc=<0 0 0>, rot=<0 0 0 1>}
 base_sleeve_bk_01_l {parent=upperarm_l, loc=<12.043 -5.9361 -2.9915>, rot=<-0.0873 -0.996 -0.0127 0.0164>}
 dyn_sleeve_bk_01_l {parent=base_sleeve_bk_01_l, loc=<0 0 0>, rot=<0 0 0 1>}
 lowerarm_l_null {parent=upperarm_l, type=Null, loc=<26.465 -0.8345 0.259>, rot=<0.0036 0.0161 -0.2298 -0.9731>}
 clavicle_r {parent=spine_05, loc=<5.4831 -0.1053 3.2946>, rot=<0.7289 0.0465 -0.6829 0.009>}
 upperarm_r {parent=clavicle_r, loc=<-9.8658 1.2136 0.4105>, rot=<0.0164 -0.3245 -0.0269 -0.9454>}
 lowerarm_r {parent=upperarm_r, loc=<-26.465 0.8345 -0.259>, rot=<0.0036 0.0161 -0.2298 -0.9731>}
 hand_r {parent=lowerarm_r, loc=<-22.7355 0.3203 0.6106>, rot=<-0.6784 -0.0223 0.0735 -0.7307>}
 thumb_01_r {parent=hand_r, loc=<-1.3118 0.3191 2.459>, rot=<0.6546 -0.4042 0.0664 -0.6354>}
 thumb_02_r {parent=thumb_01_r, loc=<-5.9336 0.0448 -0.0401>, rot=<-0.0037 0.0213 0.298 -0.9543>}
 thumb_03_r {parent=thumb_02_r, loc=<-2.2223 -0.3682 0.0907>, rot=<0.0118 -0.013 0.0066 -0.9998>}
 index_metacarpal_r {parent=hand_r, loc=<-1.6535 -0.0183 2.96>, rot=<0.0158 -0.0124 -0.0173 -0.9996>}
 index_01_r {parent=index_metacarpal_r, loc=<-8.3008 -0.0357 0.0524>, rot=<-0.0044 0.0485 0.0924 -0.9945>}
 index_02_r {parent=index_01_r, loc=<-3.3587 0.3004 -0.0452>, rot=<-0.002 0.0027 0.3225 -0.9466>}
 index_03_r {parent=index_02_r, loc=<-2.5578 0.123 0.0252>, rot=<-0.002 0.0067 0.2558 -0.9667>}
 middle_metacarpal_r {parent=hand_r, loc=<-2.1856 0.1725 0.5984>, rot=<0.006 0.0168 -0.0329 -0.9993>}
 middle_01_r {parent=middle_metacarpal_r, loc=<-7.3275 0.2447 -0.0129>, rot=<-0.0208 0.0042 0.0858 -0.9961>}
 middle_02_r {parent=middle_01_r, loc=<-4.0123 -0.1401 -0.1175>, rot=<-0.0218 0.0258 0.3512 -0.9357>}
 middle_03_r {parent=middle_02_r, loc=<-2.6077 -0.0534 -0.1373>, rot=<0.0082 0.0188 0.1708 -0.9851>}
 ring_metacarpal_r {parent=hand_r, loc=<-2.2074 0.3625 -1.3638>, rot=<-0.0284 0.0722 -0.0458 -0.9959>}
 ring_01_r {parent=ring_metacarpal_r, loc=<-7.0295 0.6197 0.1446>, rot=<-0.0489 -0.0252 0.1892 -0.9804>}
 ring_02_r {parent=ring_01_r, loc=<-3.6174 -0.4647 0.1027>, rot=<0 0 0.2806 -0.9598>}
 ring_03_r {parent=ring_02_r, loc=<-2.4152 -0.114 -0.0199>, rot=<0 0 0.1189 -0.9929>}
 pinky_metacarpal_r {parent=hand_r, loc=<-1.7616 0.9753 -3.3814>, rot=<-0.0397 0.0814 -0.069 -0.9935>}
 pinky_01_r {parent=pinky_metacarpal_r, loc=<-7.1709 0.4831 0.1094>, rot=<-0.0654 -0.0035 0.2242 -0.9723>}
 pinky_02_r {parent=pinky_01_r, loc=<-2.8274 -0.0893 0.1043>, rot=<0.0037 0.0011 0.2854 -0.9584>}
 pinky_03_r {parent=pinky_02_r, loc=<-2.2046 -0.0307 0.0078>, rot=<-0.0002 0.0098 0.1445 -0.9895>}
 deform_wrist_dn_base_r {parent=hand_r, loc=<-0.0001 1.5328 0>, rot=<0 0 0 -1>}
 deform_wrist_dn_r {parent=deform_wrist_dn_base_r, loc=<0 0 0>, rot=<0 0 0 1>}
 deform_wrist_up_base_r {parent=hand_r, loc=<-0.0002 -2.3007 -0.0001>, rot=<0 0 0 -1>}
 deform_wrist_up_r {parent=deform_wrist_up_base_r, loc=<0 0 0>, rot=<0 0 0 1>}
 lowerarm_twist_02_r {parent=lowerarm_r, loc=<-10.6707 0 0>, rot=<0 0 0 -1>}
 lowerarm_twist_01_r {parent=lowerarm_r, loc=<-18.316 0 0>, rot=<0 0 0 -1>}
 deform_elbow_in_base_r {parent=lowerarm_r, loc=<-0.1834 -3.6062 0.2572>, rot=<0 0 0 -1>}
 deform_elbow_in_r {parent=deform_elbow_in_base_r, loc=<0 0 0>, rot=<0 0 0 1>}
 deform_elbow_out_base_r {parent=lowerarm_r, loc=<0 4 -0.0001>, rot=<0 0 0 -1>}
 deform_elbow_out_r {parent=deform_elbow_out_base_r, loc=<0 0 0>, rot=<0 0 0 1>}
 upperarm_twist_01_r {parent=upperarm_r, loc=<-7.2135 0.1134 0>, rot=<0 0 0 -1>}
 deform_upperarm_bk_base_r {parent=upperarm_twist_01_r, loc=<6.4053 4.3746 -0.314>, rot=<0 0 0 -1>}
 deform_upperarm_bk_r {parent=deform_upperarm_bk_base_r, loc=<0 0 0>, rot=<0 0 0 1>}
 deform_upperarm_fr_base_r {parent=upperarm_twist_01_r, loc=<5.6312 -5.092 -0.3354>, rot=<0 0 0 -1>}
 deform_upperarm_fr_r {parent=deform_upperarm_fr_base_r, loc=<0 0 0>, rot=<0 0 0 1>}
 deform_upperarm_up_base_r {parent=upperarm_twist_01_r, loc=<8.891 -0.4825 -3.9517>, rot=<0 0 0 -1>}
 deform_upperarm_up_r {parent=deform_upperarm_up_base_r, loc=<0 0 0>, rot=<0 0 0 1>}
 upperarm_twist_02_r {parent=upperarm_r, loc=<-19.7998 0.1134 0>, rot=<0 0 0 -1>}
 base_sleeve_dw_01_r {parent=upperarm_r, loc=<-18.7974 -0.0595 10.8992>, rot=<0.9729 -0.0203 -0.1726 -0.1528>}
 dyn_sleeve_dw_01_r {parent=base_sleeve_dw_01_r, loc=<0 0 0>, rot=<0 0 0 1>}
 dyn_sleeve_dw_02_r {parent=dyn_sleeve_dw_01_r, loc=<-6.5989 0.1811 0.4109>, rot=<0 0 0 1>}
 dyn_sleeve_dw_03_r {parent=dyn_sleeve_dw_02_r, loc=<-3.5855 0 0>, rot=<0 0 0 1>}
 base_sleeve_up_01_r {parent=upperarm_r, loc=<-18.3525 0.3217 -2.6108>, rot=<0.9961 0.0061 -0.0872 -0.0106>}
 dyn_sleeve_up_01_r {parent=base_sleeve_up_01_r, loc=<0 0 0>, rot=<0 0 0 1>}
 dyn_sleeve_up_02_r {parent=dyn_sleeve_up_01_r, loc=<-8.7668 0.1811 3.2148>, rot=<0 -0.1305 0 -0.9914>}
 dyn_sleeve_up_03_r {parent=dyn_sleeve_up_02_r, loc=<-3.5855 0 0>, rot=<0 0 0 1>}
 base_sleeve_fr_01_r {parent=upperarm_r, loc=<-21.9869 7.0897 2.8945>, rot=<0.9904 -0.0447 -0.1299 -0.0168>}
 dyn_sleeve_fr_01_r {parent=base_sleeve_fr_01_r, loc=<0 0 0>, rot=<0 0 0 1>}
 dyn_sleeve_fr_02_r {parent=dyn_sleeve_fr_01_r, loc=<-7.6541 0.1811 1.4773>, rot=<0 -0.0698 0 -0.9976>}
 dyn_sleeve_fr_03_r {parent=dyn_sleeve_fr_02_r, loc=<-3.5855 0 0>, rot=<0 0 0 1>}
 base_sleeve_bk_01_r {parent=upperarm_r, loc=<-15.6063 -5.5275 4.5637>, rot=<0.9966 0.0692 -0.0443 -0.0082>}
 dyn_sleeve_bk_01_r {parent=base_sleeve_bk_01_r, loc=<0 0 0>, rot=<0 0 0 1>}
 dyn_sleeve_bk_02_r {parent=dyn_sleeve_bk_01_r, loc=<-8.2399 0.1811 0.4856>, rot=<0 -0.0698 0 -0.9976>}
 dyn_sleeve_bk_03_r {parent=dyn_sleeve_bk_02_r, loc=<-3.5855 0 0>, rot=<0 0 0 1>}
 lowerarm_r_null {parent=upperarm_r, type=Null, loc=<-26.465 0.8345 -0.259>, rot=<0.0036 0.0161 -0.2298 -0.9731>}
 upperarm_r_null {parent=clavicle_r, type=Null, loc=<-9.8658 1.2136 0.4105>, rot=<0.0164 -0.3245 -0.0269 -0.9454>}
 attach_backpack {parent=spine_05, loc=<-0.123 -13.473 0>, rot=<-0.0093 -0.707 0.0093 -0.707>}
 Backpack_BR {parent=attach_backpack, type=Null, loc=<0 3.4801 0>, rot=<0 0 0 1>}
 attach_cape {parent=spine_05, loc=<9.017 -11.714 0>, rot=<0.0616 -0.7044 -0.0616 -0.7044>}
 cape {parent=attach_cape, type=Null, loc=<0 0 0>, rot=<0 0 0 1>}
 deform_pec_aim_base_l {parent=spine_04, loc=<3.2456 9.3309 -1.2237>, rot=<0.0603 -0.4033 0.3318 -0.8507>}
 deform_pec_aim_l {parent=deform_pec_aim_base_l, loc=<0 0 0>, rot=<0 0 0 1>}
 deform_pec_base_l {parent=deform_pec_aim_l, loc=<9.3665 0 0>, rot=<0 0 0 1>}
 deform_pec_l {parent=deform_pec_base_l, loc=<0 0 0>, rot=<0 0 0 1>}
 deform_pec_aim_base_r {parent=spine_04, loc=<3.2452 9.3309 1.2237>, rot=<-0.4033 -0.0603 0.8507 0.3318>}
 deform_pec_aim_r {parent=deform_pec_aim_base_r, loc=<0 0 0>, rot=<0 0 0 1>}
 deform_pec_base_r {parent=deform_pec_aim_r, loc=<-9.3667 -0.0003 -0.0004>, rot=<0 0 0 1>}
 deform_pec_r {parent=deform_pec_base_r, loc=<0 0 0>, rot=<0 0 0 1>}
 base_katana_l {parent=spine_01, loc=<2.1697 6.7466 -9.6415>, rot=<-0.0864 -0.0114 -0.13 -0.9877>}
 dyn_katana_l {parent=base_katana_l, loc=<0 0 0>, rot=<0 0 0 1>}
 dyn_katana_blade_l {parent=dyn_katana_l, loc=<0 0 0>, rot=<0 0 0 1>}
 base_ribbon_l {parent=spine_01, loc=<0.2393 10.962 -1.1047>, rot=<0.5 -0.5 -0.5 -0.5>}
 dyn_ribbon_l {parent=base_ribbon_l, loc=<0 0 0>, rot=<0 0 0 1>}
 base_ribbon_r {parent=spine_01, loc=<0.5039 10.962 1.0312>, rot=<0.5461 0.3984 0.5219 -0.5203>}
 dyn_ribbon_r {parent=base_ribbon_r, loc=<0 0 0>, rot=<0 0 0 1>}
 thigh_l {parent=pelvis, loc=<-1.0855 -0.2109 -8.7053>, rot=<0.0009 0.0355 0.0155 -0.9993>}
 calf_l {parent=thigh_l, loc=<-42.5284 -0.0019 -0.7297>, rot=<0.0006 -0.0161 -0.0391 -0.9991>}
 foot_l {parent=calf_l, loc=<-46.7401 -2.5014 -0.592>, rot=<0.0427 -0.0238 0.0353 -0.9982>}
 ball_l {parent=foot_l, loc=<-3.5658 10.8066 -0.1544>, rot=<0.0001 -0.0001 -0.7063 -0.7079>}
 calf_twist_02_l {parent=calf_l, loc=<-19.1546 -1.7943 -0.2638>, rot=<0.0237 0.0095 -0.0035 -0.9997>}
 calf_twist_01_l {parent=calf_l, loc=<-35.8483 -1.9197 -0.5794>, rot=<0.0237 0.0095 -0.0035 -0.9997>}
 deform_knee_in_base_l {parent=calf_l, loc=<0.8031 -5.0231 0.2456>, rot=<0 0 0 -1>}
 deform_knee_in_l {parent=deform_knee_in_base_l, loc=<0 0 0>, rot=<0 0 0 1>}
 deform_knee_out_base_l {parent=calf_l, loc=<0 4 0>, rot=<0 0 0 -1>}
 deform_knee_out_l {parent=deform_knee_out_base_l, loc=<0 0 0>, rot=<0 0 0 1>}
 base_pants_in_01_l {parent=calf_l, loc=<-3.8684 -2.9874 8.0969>, rot=<0 0 0.0414 -0.9991>}
 dyn_pants_in_01_l {parent=base_pants_in_01_l, loc=<0 0 0>, rot=<0 0 0 1>}
 dyn_pants_in_02_l {parent=dyn_pants_in_01_l, loc=<-14.4518 -1.6952 -1.1893>, rot=<0 0 0 1>}
 base_pants_out_01_l {parent=calf_l, loc=<-2.1697 -3.078 -8.4362>, rot=<0 0 0.0414 -0.9991>}
 dyn_pants_out_01_l {parent=base_pants_out_01_l, loc=<0 0 0>, rot=<0 0 0 1>}
 dyn_pants_out_02_l {parent=dyn_pants_out_01_l, loc=<-16.5136 -1.6707 -3.2597>, rot=<0 0 0 1>}
 base_pants_fr_01_l {parent=calf_l, loc=<-1.669 12.2675 0.8961>, rot=<0 0 0.0414 -0.9991>}
 dyn_pants_fr_01_l {parent=base_pants_fr_01_l, loc=<0 0 0>, rot=<0 0 0 1>}
 dyn_pants_fr_02_l {parent=dyn_pants_fr_01_l, loc=<-14.3798 0.5168 -0.6285>, rot=<0 0 0 1>}
 base_pants_bk_01_l {parent=calf_l, loc=<-3.691 -18.2059 1.7059>, rot=<0 0 0 1>}
 dyn_pants_bk_01_l {parent=base_pants_bk_01_l, loc=<0 0 0>, rot=<0 0 0 1>}
 dyn_pants_bk_02_l {parent=dyn_pants_bk_01_l, loc=<-13.5862 -0.5964 -2.6198>, rot=<0 0 0 1>}
 thigh_twist_01_l {parent=thigh_l, loc=<-27.952 0 0>, rot=<0.0013 0 0 -1>}
 deform_groin_base_l {parent=thigh_l, loc=<0.4801 5.6776 -0.2486>, rot=<0 0 0 -1>}
 deform_groin_l {parent=deform_groin_base_l, loc=<0 0 0>, rot=<0 0 0 1>}
 calf_l_null {parent=pelvis, type=Null, loc=<-43.4348 1.0985 -12.4491>, rot=<-0.0027 -0.0193 0.0235 0.9995>}
 thigh_r {parent=pelvis, loc=<-1.0855 -0.2109 8.7053>, rot=<0.0355 -0.0009 0.9993 0.0155>}
 calf_r {parent=thigh_r, loc=<42.5284 0.0019 0.7297>, rot=<0.0006 -0.0161 -0.0391 -0.9991>}
 foot_r {parent=calf_r, loc=<46.7401 2.5014 0.592>, rot=<0.0427 -0.0238 0.0353 -0.9982>}
 ball_r {parent=foot_r, loc=<3.5658 -10.8066 0.1544>, rot=<0.0001 -0.0001 -0.7063 -0.7079>}
 calf_twist_02_r {parent=calf_r, loc=<19.1546 1.7943 0.2638>, rot=<0.0237 0.0095 -0.0035 -0.9997>}
 calf_twist_01_r {parent=calf_r, loc=<35.8483 1.9197 0.5794>, rot=<0.0237 0.0095 -0.0035 -0.9997>}
 deform_knee_in_base_r {parent=calf_r, loc=<-0.8031 5.0231 -0.2457>, rot=<0 0 0 -1>}
 deform_knee_in_r {parent=deform_knee_in_base_r, loc=<0 0 0>, rot=<0 0 0 1>}
 deform_knee_out_base_r {parent=calf_r, loc=<0 -4 0>, rot=<0 0 0 -1>}
 deform_knee_out_r {parent=deform_knee_out_base_r, loc=<0 0 0>, rot=<0 0 0 1>}
 base_pants_in_01_r {parent=calf_r, loc=<3.8685 2.9873 -8.0966>, rot=<0 0 0.9991 0.0414>}
 dyn_pants_in_01_r {parent=base_pants_in_01_r, loc=<0 0 0>, rot=<0 0 0 1>}
 dyn_pants_in_02_r {parent=dyn_pants_in_01_r, loc=<-14.4518 -1.6952 1.0236>, rot=<0 0 0 1>}
 base_pants_out_01_r {parent=calf_r, loc=<2.1697 3.078 8.436>, rot=<0 0 0.9991 0.0414>}
 dyn_pants_out_01_r {parent=base_pants_out_01_r, loc=<0 0 0>, rot=<0 0 0 1>}
 dyn_pants_out_02_r {parent=dyn_pants_out_01_r, loc=<-16.5136 -1.6707 3.26>, rot=<0 0 0 1>}
 base_pants_fr_01_r {parent=calf_r, loc=<1.669 -12.2675 -0.8961>, rot=<0 0 0.9991 0.0414>}
 dyn_pants_fr_01_r {parent=base_pants_fr_01_r, loc=<0 0 0>, rot=<0 0 0 1>}
 dyn_pants_fr_02_r {parent=dyn_pants_fr_01_r, loc=<-14.393 0.5183 0.1263>, rot=<0 0 0 1>}
 base_pants_bk_01_r {parent=calf_r, loc=<3.73 18.2066 -1.8725>, rot=<0 0 1 0>}
 dyn_pants_bk_01_r {parent=base_pants_bk_01_r, loc=<0 0 0>, rot=<0 0 0 1>}
 dyn_pants_bk_02_r {parent=dyn_pants_bk_01_r, loc=<-13.7402 -0.5948 3.1488>, rot=<0 0 0 1>}
 thigh_twist_01_r {parent=thigh_r, loc=<27.952 0 0>, rot=<0.0013 0 0 -1>}
 deform_groin_base_r {parent=thigh_r, loc=<-0.4801 -5.6776 0.2486>, rot=<0 0 0 -1>}
 deform_groin_r {parent=deform_groin_base_r, loc=<0 0 0>, rot=<0 0 0 1>}
 calf_r_null {parent=thigh_r, type=Null, loc=<42.5284 0.0019 0.7297>, rot=<0.0006 -0.0161 -0.0391 -0.9991>}
 deform_glute_base_l {parent=pelvis, loc=<-7.3466 -9.0213 -9.1245>, rot=<0.0172 0.031 -0.4477 -0.8935>}
 deform_glute_l {parent=deform_glute_base_l, loc=<0 0 0>, rot=<0 0 0 1>}
 deform_glute_base_r {parent=pelvis, loc=<-7.3464 -9.0213 9.1245>, rot=<0.031 -0.0172 0.8935 -0.4477>}
 deform_glute_r {parent=deform_glute_base_r, loc=<0 0 0>, rot=<0 0 0 1>}
 base_pants_m {parent=pelvis, loc=<-26.3252 -0.713 0>, rot=<0 0 0.0179 -0.9998>}
 dyn_pants_m {parent=base_pants_m, loc=<0 0 0>, rot=<0 0 0 1>}
 thigh_r_null {parent=pelvis, type=Null, loc=<-1.0855 -0.2109 8.7053>, rot=<0.0355 -0.0009 0.9993 0.0155>}
 thigh_l_null {parent=pelvis, type=Null, loc=<-1.0855 -0.2109 -8.7053>, rot=<0.0009 0.0355 0.0155 -0.9993>}
 spine_01_null {parent=pelvis, type=Null, loc=<6.1208 -0.9328 0>, rot=<0 0 0.034 -0.9994>}
 spine_05_null {parent=pelvis, type=Null, loc=<32.9717 -4.2232 0>, rot=<0 0 0.058 -0.9983>}
]
variables [
 ModifyTransforms_WeightMaximum {type=float, default=1}
 NewVar {type=double, default=0}
 WearingGauntlets {type=double, default=0}
]
