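# Anim Blueprint for the female MajorSpeech cosmetic body, adding secondary motion to its sleeves, katana, ribbons and pants.
# It copies the parent mesh pose into a gauntlet-aware Control Rig, then chains spring-based AnimDynamics and additive bone offsets, with pants weighted by an emote value.
# Sleeves branch from a cached pose, switching to a wind-enabled setup while skydiving.
# A per-bone slot blend lets custom idles override katana and right-sleeve bones before final offsets.
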AnimBlueprint /BRCosmetics/Characters/Player/Female/Medium/Bodies/F_MED_MajorSpeech/Meshes/F_MED_MajorSpeech_AnimBP
target_skeleton=&/F_MED_MajorSpeech_Skeleton
graphs [
 AnimGraph {nodes [
   Anim Dynamics - Bone: dyn_sleeve_dw_01_r!(ComponentPose, Alpha=1, GravityOverride:Vector="0,0,-960", ExternalForce="0,0,0") -> (Pose) {
    angularDampingOverride=1
    angularSpringConstant=9000
    bAngularSpring=true
    bEnableWind=true
    bOverrideAngularDamping=true
    bUseGravityOverride=true
    gravityScale=0.35
    simulationSpace=RootRelative
    boundBone {boneName=dyn_sleeve_dw_01_r}
    componentAppliedLinearAccClamp=<2000 2000 2000>
    componentLinearAccScale=<0.6 0.6 0.6>
    componentLinearVelScale=<0.5 0.5 0.5>
    gravityOverride {z=-960}
    physicsBodyDefinitions [{
      boundBone {boneName=dyn_sleeve_dw_01_r}
      boxExtents=<20 30 30>
      constraintSetup {angularLimitsMax=<0 0 0>, angularLimitsMin=<0 0 0>, angularTarget=<1 0 0>}
      localJointOffset=<3.6 0 0>
     }]
   }
   "Transform (Modify) Bone - Bone: base_katana_l"!(ComponentPose, bAlphaBoolEnabled:bool=True, Translation="-2.908,-3.450,0.286", Rotation="-22.276,7.603,3.108", Scale="1,1,1") -> (Pose) {
    alphaInputType=Bool
    rotationMode=BMM_Additive
    rotationSpace=BCS_BoneSpace
    translationMode=BMM_Additive
    translationSpace=BCS_BoneSpace
    alphaBoolBlend {blendInTime=0.25, blendOutTime=0.25}
    boneToModify {boneName=base_katana_l}
    rotation=<rot -22.276 7.603 3.108>
    translation=<-2.908 -3.45 0.286>
   }
   "Transform (Modify) Bone - Bone: base_katana_l"(bAlphaBoolEnabled=True, Translation="-1.5,-1.5,-1.5", Rotation="0.8,-5,15", Scale="1,1,1") {
    alphaInputType=Bool
    rotationMode=BMM_Additive
    rotationSpace=BCS_BoneSpace
    translationMode=BMM_Additive
    translationSpace=BCS_BoneSpace
    alphaBoolBlend {blendInTime=0.25, blendOutTime=0.25}
    boneToModify {boneName=base_katana_l}
    rotation=<rot 0.8 -5 15>
    translation=<-1.5 -1.5 -1.5>
   }
   n1 = Anim Dynamics - Bone: dyn_sleeve_dw_01_l(ComponentPose=Local To Component(LocalPose=Use cached pose 'Non sleeve dynamics'()), Alpha=1, ExternalForce="0,0,0") {
    angularDampingOverride=1
    angularSpringConstant=20000
    bAngularSpring=true
    bOverrideAngularDamping=true
    bUseGravityOverride=true
    gravityScale=0.35
    simulationSpace=RootRelative
    boundBone {boneName=dyn_sleeve_dw_01_l}
    componentAppliedLinearAccClamp=<3000 3000 3000>
    componentLinearAccScale=<1 1 1>
    componentLinearVelScale=<1 1 1>
    gravityOverride {z=-980}
    physicsBodyDefinitions [{
      boundBone {boneName=dyn_sleeve_dw_01_l}
      boxExtents=<20 30 30>
      constraintSetup {angularLimitsMax=<15 15 15>, angularLimitsMin=<-15 -15 -15>, angularTarget=<1 0 0>}
      localJointOffset=<4 0 0>
     }]
   }
   n2 = Anim Dynamics - Bone: dyn_sleeve_up_01_l(ComponentPose=@n1, Alpha=1, ExternalForce="0,0,0") {
    angularDampingOverride=1
    angularSpringConstant=20000
    bAngularSpring=true
    bOverrideAngularDamping=true
    bUseGravityOverride=true
    gravityScale=0.35
    simulationSpace=RootRelative
    boundBone {boneName=dyn_sleeve_up_01_l}
    componentAppliedLinearAccClamp=<3000 3000 3000>
    componentLinearAccScale=<1 1 1>
    componentLinearVelScale=<1 1 1>
    gravityOverride {z=-980}
    physicsBodyDefinitions [{
      boundBone {boneName=dyn_sleeve_up_01_l}
      boxExtents=<20 30 30>
      constraintSetup {angularLimitsMax=<15 15 15>, angularLimitsMin=<-15 -15 -15>, angularTarget=<1 0 0>}
      localJointOffset=<4 0 0>
     }]
   }
   n3 = Anim Dynamics - Bone: dyn_sleeve_fr_01_l(ComponentPose=@n2, Alpha=1, ExternalForce="0,0,0") {
    angularDampingOverride=1
    angularSpringConstant=20000
    bAngularSpring=true
    bOverrideAngularDamping=true
    bUseGravityOverride=true
    gravityScale=0.35
    simulationSpace=RootRelative
    boundBone {boneName=dyn_sleeve_fr_01_l}
    componentAppliedLinearAccClamp=<3000 3000 3000>
    componentLinearAccScale=<1 1 1>
    componentLinearVelScale=<1 1 1>
    gravityOverride {z=-980}
    physicsBodyDefinitions [{
      boundBone {boneName=dyn_sleeve_fr_01_l}
      boxExtents=<20 30 30>
      constraintSetup {angularLimitsMax=<15 15 15>, angularLimitsMin=<-15 -15 -15>, angularTarget=<1 0 0>}
      localJointOffset=<4 0 0>
     }]
   }
   n4 = Anim Dynamics - Bone: dyn_sleeve_bk_01_l(ComponentPose=@n3, Alpha=1, ExternalForce="0,0,0") {
    angularDampingOverride=1
    angularSpringConstant=20000
    bAngularSpring=true
    bOverrideAngularDamping=true
    bUseGravityOverride=true
    gravityScale=0.35
    simulationSpace=RootRelative
    boundBone {boneName=dyn_sleeve_bk_01_l}
    componentAppliedLinearAccClamp=<3000 3000 3000>
    componentLinearAccScale=<1 1 1>
    componentLinearVelScale=<1 1 1>
    gravityOverride {z=-980}
    physicsBodyDefinitions [{
      boundBone {boneName=dyn_sleeve_bk_01_l}
      boxExtents=<20 30 30>
      constraintSetup {angularLimitsMax=<15 15 15>, angularLimitsMin=<-15 -15 -15>, angularTarget=<1 0 0>}
      localJointOffset=<4 0 0>
     }]
   }
   n5 = Anim Dynamics - Bone: dyn_sleeve_dw_01_r!(ComponentPose=@n4, Alpha=1, ExternalForce="0,0,0") -> (Pose) {
    angularDampingOverride=1
    angularSpringConstant=9000
    bAngularSpring=true
    bOverrideAngularDamping=true
    bUseGravityOverride=true
    gravityScale=0.35
    simulationSpace=RootRelative
    boundBone {boneName=dyn_sleeve_dw_01_r}
    componentAppliedLinearAccClamp=<2000 2000 2000>
    componentLinearAccScale=<0.6 0.6 0.6>
    componentLinearVelScale=<0.5 0.5 0.5>
    gravityOverride {z=-960}
    physicsBodyDefinitions [{
      boundBone {boneName=dyn_sleeve_dw_01_r}
      boxExtents=<20 30 30>
      constraintSetup {angularLimitsMax=<0 0 5>, angularLimitsMin=<0 -1 0>, angularTarget=<1 0 0>}
      localJointOffset=<3.6 0 0>
     }]
   }
   n6 = Anim Dynamics - Bone: dyn_sleeve_dw_02_r(ComponentPose=@n5, Alpha=1, ExternalForce="0,0,0") {
    angularDampingOverride=1
    angularSpringConstant=9000
    bAngularSpring=true
    bOverrideAngularDamping=true
    bUseGravityOverride=true
    gravityScale=0.35
    simulationSpace=RootRelative
    boundBone {boneName=dyn_sleeve_dw_02_r}
    componentAppliedLinearAccClamp=<2000 2000 2000>
    componentLinearAccScale=<0.6 0.6 0.6>
    componentLinearVelScale=<0.5 0.5 0.5>
    gravityOverride {z=-960}
    physicsBodyDefinitions [{
      boundBone {boneName=dyn_sleeve_dw_02_r}
      boxExtents=<20 30 30>
      constraintSetup {angularLimitsMax=<0 0 10>, angularLimitsMin=<0 -5 0>, angularTarget=<1 0 0>}
      localJointOffset=<3.6 0 0>
     }]
    simSpaceSettings {maxAngularAcceleration=100, maxAngularVelocity=10, simSpaceAngularAlpha=1}
   }
   n7 = Anim Dynamics - Bone: dyn_sleeve_up_02_r(ComponentPose=@n6, Alpha=1, ExternalForce="0,0,0") {
    angularDampingOverride=1
    angularSpringConstant=12000
    bAngularSpring=true
    bOverrideAngularDamping=true
    bUseGravityOverride=true
    gravityScale=0.35
    simulationSpace=RootRelative
    boundBone {boneName=dyn_sleeve_up_02_r}
    componentAppliedLinearAccClamp=<2000 2000 2000>
    componentLinearAccScale=<0.6 0.6 0.6>
    componentLinearVelScale=<0.5 0.5 0.5>
    gravityOverride {z=-980}
    physicsBodyDefinitions [{
      boundBone {boneName=dyn_sleeve_up_02_r}
      boxExtents=<20 30 30>
      constraintSetup {angularLimitsMax=<0 20 3>, angularLimitsMin=<0 0 -3>, angularTarget=<1 0 0>}
      localJointOffset=<3.6 0 0>
     }]
    simSpaceSettings {maxAngularAcceleration=100, maxAngularVelocity=10, simSpaceAngularAlpha=1}
   }
   n8 = Anim Dynamics - Bone: dyn_sleeve_fr_02_r(ComponentPose=@n7, Alpha=1, ExternalForce="0,0,0") {
    angularDampingOverride=1
    angularSpringConstant=12000
    bAngularSpring=true
    bOverrideAngularDamping=true
    bUseGravityOverride=true
    gravityScale=0.35
    simulationSpace=RootRelative
    boundBone {boneName=dyn_sleeve_fr_02_r}
    componentAppliedLinearAccClamp=<2000 2000 2000>
    componentLinearAccScale=<0.6 0.6 0.6>
    componentLinearVelScale=<0.5 0.5 0.5>
    gravityOverride {z=-980}
    physicsBodyDefinitions [{
      boundBone {boneName=dyn_sleeve_fr_02_r}
      boxExtents=<20 30 30>
      constraintSetup {angularLimitsMax=<0 5 20>, angularLimitsMin=<0 -1 0>, angularTarget=<1 0 0>}
      localJointOffset=<3.6 0 0>
     }]
    simSpaceSettings {maxAngularAcceleration=100, maxAngularVelocity=10, simSpaceAngularAlpha=1}
   }
   n9 = Anim Dynamics - Bone: dyn_sleeve_bk_01_r(ComponentPose=@n8, Alpha=1, ExternalForce="0,0,0") {
    angularDampingOverride=1
    angularSpringConstant=9000
    bAngularSpring=true
    bOverrideAngularDamping=true
    bUseGravityOverride=true
    gravityScale=0.35
    simulationSpace=RootRelative
    boundBone {boneName=dyn_sleeve_bk_01_r}
    componentAppliedLinearAccClamp=<2000 2000 2000>
    componentLinearAccScale=<0.5 0.5 0.5>
    componentLinearVelScale=<0.5 0.5 0.5>
    gravityOverride {z=-960}
    physicsBodyDefinitions [{
      boundBone {boneName=dyn_sleeve_bk_01_r}
      boxExtents=<20 30 30>
      constraintSetup {angularLimitsMax=<0 0 0>, angularLimitsMin=<0 0 -2>, angularTarget=<1 0 0>}
      localJointOffset=<3.5 0 0>
     }]
   }
   n10 = Anim Dynamics - Bone: dyn_sleeve_bk_02_r(ComponentPose=@n9, Alpha=1, ExternalForce="0,0,0") {
    angularDampingOverride=1
    angularSpringConstant=9000
    bAngularSpring=true
    bOverrideAngularDamping=true
    bUseGravityOverride=true
    gravityScale=0.35
    simulationSpace=RootRelative
    boundBone {boneName=dyn_sleeve_bk_02_r}
    componentAppliedLinearAccClamp=<2000 2000 2000>
    componentLinearAccScale=<0.6 0.6 0.6>
    componentLinearVelScale=<0.5 0.5 0.5>
    gravityOverride {z=-960}
    physicsBodyDefinitions [{
      boundBone {boneName=dyn_sleeve_bk_02_r}
      boxExtents=<20 30 30>
      constraintSetup {angularLimitsMax=<0 0 0>, angularLimitsMin=<0 0 -15>, angularTarget=<1 0 0>}
      localJointOffset=<3.5 0 0>
     }]
    simSpaceSettings {maxAngularAcceleration=100, maxAngularVelocity=10, simSpaceAngularAlpha=1}
   }
   n11 = "Transform (Modify) Bone - Bone: base_sleeve_bk_01_r"!(ComponentPose=@n10, bAlphaBoolEnabled:bool=True, Translation="2,3,0", Rotation="0,-8,0", Scale="1,1,1") -> (Pose) {
    alphaInputType=Bool
    rotationMode=BMM_Additive
    rotationSpace=BCS_BoneSpace
    translationMode=BMM_Additive
    translationSpace=BCS_BoneSpace
    alphaBoolBlend {blendInTime=0.4, blendOutTime=0.7}
    boneToModify {boneName=base_sleeve_bk_01_r}
    rotation {yaw=-8}
    translation=<2 3>
   }
   n12 = "Transform (Modify) Bone - Bone: base_sleeve_up_01_l"!(ComponentPose=@n11, bAlphaBoolEnabled:bool=True, Translation="3.6,-1.8,2.5", Rotation="2,7,40", Scale="1,1,1") -> (Pose) {
    alphaInputType=Bool
    rotationMode=BMM_Additive
    rotationSpace=BCS_BoneSpace
    translationMode=BMM_Additive
    translationSpace=BCS_BoneSpace
    boneToModify {boneName=base_sleeve_up_01_l}
    rotation=<rot 2 7 40>
    translation=<3.6 -1.8 2.5>
   }
   n13 = "Transform (Modify) Bone - Bone: base_sleeve_fr_01_l1"!(ComponentPose=@n12, bAlphaBoolEnabled:bool=True, Translation="0.5,0,0", Rotation="0,0,0", Scale="1,1,1") -> (Pose) {
    alphaInputType=Bool
    rotationMode=BMM_Additive
    rotationSpace=BCS_BoneSpace
    translationMode=BMM_Additive
    translationSpace=BCS_BoneSpace
    boneToModify {boneName=base_sleeve_fr_01_l1}
    translation {x=0.5}
   }
   n14 = "Transform (Modify) Bone - Bone: base_sleeve_dw_01_r"!(ComponentPose=@n13, bAlphaBoolEnabled:bool=True, Translation="-1.5,0,3", Rotation="-10,6,0", Scale="1,0.7,1") -> (Pose) {
    alphaInputType=Bool
    rotationMode=BMM_Additive
    rotationSpace=BCS_BoneSpace
    scaleMode=BMM_Additive
    scaleSpace=BCS_BoneSpace
    translationMode=BMM_Additive
    translationSpace=BCS_BoneSpace
    boneToModify {boneName=base_sleeve_dw_01_r}
    rotation {pitch=-10, yaw=6}
    scale {y=0.7}
    translation {x=-1.5, z=3}
   }
   n15 = "Transform (Modify) Bone - Bone: base_sleeve_up_01_r"!(ComponentPose=@n14, bAlphaBoolEnabled:bool=True, Translation="0,0,0", Rotation="-1.5,0,0", Scale="1,1,1") -> (Pose) {
    alphaInputType=Bool
    rotationMode=BMM_Additive
    rotationSpace=BCS_BoneSpace
    scaleMode=BMM_Additive
    scaleSpace=BCS_BoneSpace
    translationMode=BMM_Additive
    translationSpace=BCS_BoneSpace
    boneToModify {boneName=base_sleeve_up_01_r}
    rotation {pitch=-1.5}
   }
   n16 = "Transform (Modify) Bone - Bone: base_sleeve_fr_01_r"!(ComponentPose=@n15, bAlphaBoolEnabled:bool=True, Translation="-2,4,0", Rotation="0,-35,0", Scale="1,1,1") -> (Pose) {
    alphaInputType=Bool
    rotationMode=BMM_Additive
    rotationSpace=BCS_BoneSpace
    scaleMode=BMM_Additive
    scaleSpace=BCS_BoneSpace
    translationMode=BMM_Additive
    translationSpace=BCS_BoneSpace
    boneToModify {boneName=base_sleeve_fr_01_r}
    rotation {yaw=-35}
    translation=<-2 4>
   }
   n17 = "Transform (Modify) Bone - Bone: base_sleeve_bk_01_r"(ComponentPose=@n16, bAlphaBoolEnabled=True, Translation="0,4,0", Rotation="0,47,0", Scale="1,1,0.1") {
    alphaInputType=Bool
    rotationMode=BMM_Additive
    rotationSpace=BCS_BoneSpace
    scaleMode=BMM_Additive
    scaleSpace=BCS_BoneSpace
    translationMode=BMM_Additive
    translationSpace=BCS_BoneSpace
    boneToModify {boneName=base_sleeve_bk_01_r}
    rotation {yaw=47}
    scale {z=0.1}
    translation {y=4}
   }
   n18 = Anim Dynamics - Bone: dyn_sleeve_dw_01_l!(
    ComponentPose=Local To Component(LocalPose=Use cached pose 'Non sleeve dynamics'())
    Alpha=1
    GravityOverride:Vector="0,0,-980"
    ExternalForce="0,0,0"
   ) -> (Pose) {
    angularDampingOverride=1
    angularSpringConstant=20000
    bAngularSpring=true
    bOverrideAngularDamping=true
    bUseGravityOverride=true
    gravityScale=0.35
    simulationSpace=RootRelative
    boundBone {boneName=dyn_sleeve_dw_01_l}
    componentAppliedLinearAccClamp=<3000 3000 3000>
    componentLinearAccScale=<1 1 1>
    componentLinearVelScale=<1 1 1>
    gravityOverride {z=-980}
    physicsBodyDefinitions [{
      boundBone {boneName=dyn_sleeve_dw_01_l}
      boxExtents=<20 30 30>
      constraintSetup {angularLimitsMax=<0 25 2>, angularLimitsMin=<0 0 -1>, angularTarget=<1 0 0>}
      localJointOffset=<4 0 0>
     }]
   }
   n19 = Anim Dynamics - Bone: dyn_sleeve_up_01_l!(ComponentPose=@n18, Alpha=1, GravityOverride:Vector="0,0,-980", ExternalForce="0,0,0") -> (Pose) {
    angularDampingOverride=1
    angularSpringConstant=20000
    bAngularSpring=true
    bOverrideAngularDamping=true
    bUseGravityOverride=true
    gravityScale=0.35
    simulationSpace=RootRelative
    boundBone {boneName=dyn_sleeve_up_01_l}
    componentAppliedLinearAccClamp=<3000 3000 3000>
    componentLinearAccScale=<1 1 1>
    componentLinearVelScale=<1 1 1>
    gravityOverride {z=-980}
    physicsBodyDefinitions [{
      boundBone {boneName=dyn_sleeve_up_01_l}
      boxExtents=<20 30 30>
      constraintSetup {angularLimitsMax=<0 -1 1>, angularLimitsMin=<0 -20 -1>, angularTarget=<1 0 0>}
      localJointOffset=<4 0 0>
     }]
   }
   n20 = Anim Dynamics - Bone: dyn_sleeve_fr_01_l!(ComponentPose=@n19, Alpha=1, GravityOverride:Vector="0,0,-980", ExternalForce="0,0,0") -> (Pose) {
    angularDampingOverride=1
    angularSpringConstant=20000
    bAngularSpring=true
    bOverrideAngularDamping=true
    bUseGravityOverride=true
    gravityScale=0.35
    simulationSpace=RootRelative
    boundBone {boneName=dyn_sleeve_fr_01_l}
    componentAppliedLinearAccClamp=<3000 3000 3000>
    componentLinearAccScale=<1 1 1>
    componentLinearVelScale=<1 1 1>
    gravityOverride {z=-980}
    physicsBodyDefinitions [{
      boundBone {boneName=dyn_sleeve_fr_01_l}
      boxExtents=<20 30 30>
      constraintSetup {angularLimitsMax=<0 1 2>, angularLimitsMin=<0 -1 -25>, angularTarget=<1 0 0>}
      localJointOffset=<4 0 0>
     }]
   }
   n21 = Anim Dynamics - Bone: dyn_sleeve_bk_01_l!(ComponentPose=@n20, Alpha=1, GravityOverride:Vector="0,0,-980", ExternalForce="0,0,0") -> (Pose) {
    angularDampingOverride=1
    angularSpringConstant=20000
    bAngularSpring=true
    bOverrideAngularDamping=true
    bUseGravityOverride=true
    gravityScale=0.35
    simulationSpace=RootRelative
    boundBone {boneName=dyn_sleeve_bk_01_l}
    componentAppliedLinearAccClamp=<3000 3000 3000>
    componentLinearAccScale=<1 1 1>
    componentLinearVelScale=<1 1 1>
    gravityOverride {z=-980}
    physicsBodyDefinitions [{
      boundBone {boneName=dyn_sleeve_bk_01_l}
      boxExtents=<20 30 30>
      constraintSetup {angularLimitsMax=<0 1 25>, angularLimitsMin=<0 -1 -2>, angularTarget=<1 0 0>}
      localJointOffset=<4 0 0>
     }]
   }
   n22 = Anim Dynamics - Bone: dyn_sleeve_dw_02_r!(ComponentPose=@n21, Alpha=1, GravityOverride:Vector="0,0,-960", ExternalForce="0,0,0") -> (Pose) {
    angularDampingOverride=1
    angularSpringConstant=9000
    bAngularSpring=true
    bEnableWind=true
    bOverrideAngularDamping=true
    bUseGravityOverride=true
    gravityScale=0.35
    simulationSpace=RootRelative
    boundBone {boneName=dyn_sleeve_dw_02_r}
    componentAppliedLinearAccClamp=<2000 2000 2000>
    componentLinearAccScale=<0.6 0.6 0.6>
    componentLinearVelScale=<0.5 0.5 0.5>
    gravityOverride {z=-960}
    physicsBodyDefinitions [{
      boundBone {boneName=dyn_sleeve_dw_02_r}
      boxExtents=<20 30 30>
      constraintSetup {angularLimitsMax=<0 0 0>, angularLimitsMin=<0 0 0>, angularTarget=<1 0 0>}
      localJointOffset=<3.6 0 0>
     }]
   }
   n23 = Anim Dynamics - Bone: dyn_sleeve_up_01_r!(ComponentPose=@n22, Alpha=1, GravityOverride:Vector="0,0,-960", ExternalForce="0,0,0") -> (Pose) {
    angularDampingOverride=1
    angularSpringConstant=9000
    bAngularSpring=true
    bEnableWind=true
    bOverrideAngularDamping=true
    bUseGravityOverride=true
    gravityScale=0.35
    simulationSpace=RootRelative
    boundBone {boneName=dyn_sleeve_up_01_r}
    componentAppliedLinearAccClamp=<2000 2000 2000>
    componentLinearAccScale=<0.6 0.6 0.6>
    componentLinearVelScale=<0.5 0.5 0.5>
    gravityOverride {z=-960}
    physicsBodyDefinitions [{
      boundBone {boneName=dyn_sleeve_up_01_r}
      boxExtents=<20 30 30>
      constraintSetup {angularLimitsMax=<0 10 0>, angularLimitsMin=<0 0 0>, angularTarget=<1 0 0>}
      localJointOffset=<3.6 0 0>
     }]
   }
   n24 = Anim Dynamics - Bone: dyn_sleeve_up_02_r!(ComponentPose=@n23, Alpha=1, GravityOverride:Vector="0,0,-960", ExternalForce="0,0,0") -> (Pose) {
    angularDampingOverride=1
    angularSpringConstant=9000
    bAngularSpring=true
    bEnableWind=true
    bOverrideAngularDamping=true
    bUseGravityOverride=true
    gravityScale=0.35
    simulationSpace=RootRelative
    boundBone {boneName=dyn_sleeve_up_02_r}
    componentAppliedLinearAccClamp=<2000 2000 2000>
    componentLinearAccScale=<0.6 0.6 0.6>
    componentLinearVelScale=<0.5 0.5 0.5>
    gravityOverride {z=-960}
    physicsBodyDefinitions [{
      boundBone {boneName=dyn_sleeve_up_02_r}
      boxExtents=<20 30 30>
      constraintSetup {angularLimitsMax=<0 25 0>, angularLimitsMin=<0 0 0>, angularTarget=<1 0 0>}
      localJointOffset=<3.6 0 0>
     }]
   }
   n25 = Anim Dynamics - Bone: dyn_sleeve_fr_01_r!(ComponentPose=@n24, Alpha=1, GravityOverride:Vector="0,0,-960", ExternalForce="0,0,0") -> (Pose) {
    angularDampingOverride=1
    angularSpringConstant=9000
    bAngularSpring=true
    bEnableWind=true
    bOverrideAngularDamping=true
    bUseGravityOverride=true
    gravityScale=0.35
    simulationSpace=RootRelative
    boundBone {boneName=dyn_sleeve_fr_01_r}
    componentAppliedLinearAccClamp=<2000 2000 2000>
    componentLinearAccScale=<0.6 0.6 0.6>
    componentLinearVelScale=<0.5 0.5 0.5>
    gravityOverride {z=-960}
    physicsBodyDefinitions [{
      boundBone {boneName=dyn_sleeve_fr_01_r}
      boxExtents=<20 30 30>
      constraintSetup {angularLimitsMax=<0 0 0>, angularLimitsMin=<0 0 -10>, angularTarget=<1 0 0>}
      localJointOffset=<3.6 0 0>
     }]
   }
   n26 = Anim Dynamics - Bone: dyn_sleeve_fr_02_r!(ComponentPose=@n25, Alpha=1, GravityOverride:Vector="0,0,-960", ExternalForce="0,0,0") -> (Pose) {
    angularDampingOverride=1
    angularSpringConstant=9000
    bAngularSpring=true
    bEnableWind=true
    bOverrideAngularDamping=true
    bUseGravityOverride=true
    gravityScale=0.35
    simulationSpace=RootRelative
    boundBone {boneName=dyn_sleeve_fr_02_r}
    componentAppliedLinearAccClamp=<2000 2000 2000>
    componentLinearAccScale=<0.6 0.6 0.6>
    componentLinearVelScale=<0.5 0.5 0.5>
    gravityOverride {z=-960}
    physicsBodyDefinitions [{
      boundBone {boneName=dyn_sleeve_fr_02_r}
      boxExtents=<20 30 30>
      constraintSetup {angularLimitsMax=<0 0 0>, angularLimitsMin=<0 0 -25>, angularTarget=<1 0 0>}
      localJointOffset=<3.6 0 0>
     }]
   }
   n27 = Anim Dynamics - Bone: dyn_sleeve_bk_01_r!(ComponentPose=@n26, Alpha=1, GravityOverride:Vector="0,0,-960", ExternalForce="0,0,0") -> (Pose) {
    angularDampingOverride=1
    angularSpringConstant=9000
    bAngularSpring=true
    bEnableWind=true
    bOverrideAngularDamping=true
    bUseGravityOverride=true
    gravityScale=0.35
    simulationSpace=RootRelative
    boundBone {boneName=dyn_sleeve_bk_01_r}
    componentAppliedLinearAccClamp=<2000 2000 2000>
    componentLinearAccScale=<0.5 0.5 0.5>
    componentLinearVelScale=<0.5 0.5 0.5>
    gravityOverride {z=-960}
    physicsBodyDefinitions [{
      boundBone {boneName=dyn_sleeve_bk_01_r}
      boxExtents=<20 30 30>
      constraintSetup {angularLimitsMax=<0 0 0>, angularLimitsMin=<0 0 -3>, angularTarget=<1 0 0>}
      localJointOffset=<3.5 0 0>
     }]
   }
   n28 = Anim Dynamics - Bone: dyn_sleeve_bk_02_r!(ComponentPose=@n27, Alpha=1, GravityOverride:Vector="0,0,-960", ExternalForce="0,0,0") -> (Pose) {
    angularDampingOverride=1
    angularSpringConstant=9000
    bAngularSpring=true
    bEnableWind=true
    bOverrideAngularDamping=true
    bUseGravityOverride=true
    gravityScale=0.35
    simulationSpace=RootRelative
    boundBone {boneName=dyn_sleeve_bk_02_r}
    componentAppliedLinearAccClamp=<2000 2000 2000>
    componentLinearAccScale=<0.6 0.6 0.6>
    componentLinearVelScale=<0.5 0.5 0.5>
    gravityOverride {z=-960}
    physicsBodyDefinitions [{
      boundBone {boneName=dyn_sleeve_bk_02_r}
      boxExtents=<20 30 30>
      constraintSetup {angularLimitsMax=<0 0 0>, angularLimitsMin=<0 0 -15>, angularTarget=<1 0 0>}
      localJointOffset=<3.5 0 0>
     }]
   }
   n29 = "Transform (Modify) Bone - Bone: dyn_sleeve_up_01_r"!(ComponentPose=@n28, bAlphaBoolEnabled:bool=True, Translation="0,0,0.7", Rotation="-7,0,0", Scale="1,1,1") -> (Pose) {
    alphaInputType=Bool
    rotationMode=BMM_Additive
    rotationSpace=BCS_BoneSpace
    translationMode=BMM_Additive
    translationSpace=BCS_BoneSpace
    alphaBoolBlend {blendInTime=0.4, blendOutTime=0.7}
    boneToModify {boneName=dyn_sleeve_up_01_r}
    rotation {pitch=-7}
    translation {z=0.7}
   }
   n30 = "Transform (Modify) Bone - Bone: dyn_sleeve_fr_01_r"!(ComponentPose=@n29, bAlphaBoolEnabled:bool=True, Translation="0,0,1", Rotation="0,0,0", Scale="1,1,1") -> (Pose) {
    alphaInputType=Bool
    translationMode=BMM_Additive
    translationSpace=BCS_BoneSpace
    alphaBoolBlend {blendInTime=0.4, blendOutTime=0.7}
    boneToModify {boneName=dyn_sleeve_fr_01_r}
    translation {z=1}
   }
   n31 = Layered blend per bone(
    BasePose=Blend Poses by bool(BlendPose_0=Component To Local(ComponentPose=@n30), BlendPose_1=Component To Local(ComponentPose=@n17), bActiveValue=Get bIsSkyDiving())
    BlendPoses_0=Slot 'DefaultSlot'()
    BlendWeights_0=Get Custom Idle Is Playing Float() -> (Custom Idle Is Playing Float:double)
   ) {layerSetup [{branchFilters [
       {boneName=base_katana_l}
       {boneName=dyn_katana_l}
       {boneName=dyn_katana_blade_l}
       {boneName=base_sleeve_dw_01_r}
       {boneName=dyn_sleeve_dw_01_r}
       {boneName=dyn_sleeve_dw_02_r}
       {boneName=dyn_sleeve_dw_03_r}
       {boneName=base_sleeve_up_01_r}
       {boneName=dyn_sleeve_up_01_r}
       {boneName=dyn_sleeve_up_02_r}
       {boneName=dyn_sleeve_up_03_r}
       {boneName=base_sleeve_fr_01_r}
       {boneName=dyn_sleeve_fr_01_r}
       {boneName=dyn_sleeve_fr_02_r}
       {boneName=dyn_sleeve_fr_03_r}
       {boneName=base_sleeve_bk_01_r}
       {boneName=dyn_sleeve_bk_01_r}
       {boneName=dyn_sleeve_bk_02_r}
       {boneName=dyn_sleeve_bk_03_r}
      ]}]}
   n32 = "Transform (Modify) Bone - Bone: base_katana_l"!(ComponentPose=Local To Component(LocalPose=@n31), Alpha:float=1, Translation="0.222,1,-0.85", Rotation="-1.952,1.041,-9.937", Scale="1,1,1") -> (Pose) {
    rotationMode=BMM_Additive
    rotationSpace=BCS_BoneSpace
    translationMode=BMM_Additive
    translationSpace=BCS_BoneSpace
    boneToModify {boneName=base_katana_l}
    rotation=<rot -1.952 1.041 -9.937>
    translation=<0.222 1 -0.85>
   }
   n33 = "Transform (Modify) Bone - Bone: dyn_katana_blade_l"(ComponentPose=@n32, Alpha=1, Translation="0,0,0", Rotation="-5.110,2.796,-10.237", Scale="0.89,0.9,0.9") {
    rotationMode=BMM_Additive
    rotationSpace=BCS_BoneSpace
    scaleMode=BMM_Additive
    scaleSpace=BCS_BoneSpace
    translationMode=BMM_Additive
    translationSpace=BCS_BoneSpace
    boneToModify {boneName=dyn_katana_blade_l}
    rotation=<rot -5.11 2.796 -10.237>
    scale=<0.89 0.9 0.9>
   }
   n34 = "Transform (Modify) Bone - Bone: base_sleeve_up_01_r"!(ComponentPose=@n33, Alpha:float=1, Translation="0.35,-0.9,0.7", Rotation="0,-0.5,0", Scale="1,1,1") -> (Pose) {
    rotationMode=BMM_Additive
    rotationSpace=BCS_BoneSpace
    translationMode=BMM_Additive
    translationSpace=BCS_BoneSpace
    boneToModify {boneName=base_sleeve_up_01_r}
    rotation {yaw=-0.5}
    translation=<0.35 -0.9 0.7>
   }
   n35 = "Transform (Modify) Bone - Bone: base_sleeve_fr_01_r"!(ComponentPose=@n34, Alpha:float=1, Translation="0.8,0,0", Rotation="0,0,0", Scale="1,1,1") -> (Pose) {
    rotationMode=BMM_Additive
    rotationSpace=BCS_BoneSpace
    translationMode=BMM_Additive
    translationSpace=BCS_BoneSpace
    boneToModify {boneName=base_sleeve_fr_01_r}
    translation {x=0.8}
   }
   Output Pose(Result=Component To Local(ComponentPose=@n35)) {layerGroup=None}
   n36 = Get Playing emote() -> (Playing emote:double)
   n37 = Get Playing emote()
   n38 = Get Playing emote()
   n39 = Control Rig!(
    Source=Copy Pose From Mesh() {bUseAttachedParent=true}
    Alpha
    WearingGauntlets:double=Get IsWearingGauntlet_alpha_Right() -> (IsWearingGauntlet_alpha_Right:double)
   ) -> (Pose) {controlRigAssetReference {blueprintRigClass=&/F_MED_MajorSpeech_CtrlRig.F_MED_MajorSpeech_CtrlRig_C}}
   n40 = Anim Dynamics - Bone: dyn_katana_l(ComponentPose=Local To Component(LocalPose=@n39), Alpha=1, ExternalForce="0,0,0") {
    angularDampingOverride=1
    angularSpringConstant=13000
    bAngularSpring=true
    bOverrideAngularDamping=true
    bUseGravityOverride=true
    gravityScale=0.35
    simulationSpace=RootRelative
    alphaBoolBlend {blendInTime=0.3, blendOutTime=0.3}
    boundBone {boneName=dyn_katana_l}
    componentAppliedLinearAccClamp=<1000 1000 1000>
    componentLinearAccScale=<0.5 0.5 0.5>
    componentLinearVelScale=<0.6 0.6 0.6>
    gravityOverride {z=-980}
    physicsBodyDefinitions [{
      boundBone {boneName=dyn_katana_l}
      boxExtents=<20 27 27>
      constraintSetup {angularLimitsMax=<0 2 6>, angularLimitsMin=<0 -2 -1.5>, angularTarget=<1 0 0>}
      localJointOffset=<3.5 0 0>
     }]
   }
   n41 = "Transform (Modify) Bone - Bone: base_katana_l"!(ComponentPose=@n40, bAlphaBoolEnabled:bool=True, Translation="0,0,0", Rotation="0,10,0", Scale="1,1,1") -> (Pose) {
    alphaInputType=Bool
    rotationMode=BMM_Additive
    rotationSpace=BCS_BoneSpace
    alphaBoolBlend {blendInTime=0.25, blendOutTime=0.25}
    boneToModify {boneName=base_katana_l}
    rotation {yaw=10}
   }
   n42 = "Transform (Modify) Bone - Bone: base_katana_l"(ComponentPose=@n41, bAlphaBoolEnabled=True, Translation="0,0,0", Rotation="-4,58,12", Scale="1,1,1") {
    alphaInputType=Bool
    rotationMode=BMM_Additive
    rotationSpace=BCS_BoneSpace
    alphaBoolBlend {blendInTime=0.3, blendOutTime=0.3}
    boneToModify {boneName=base_katana_l}
    rotation=<rot -4 58 12>
   }
   n43 = "Transform (Modify) Bone - Bone: dyn_katana_blade_l"!(ComponentPose=@n42, bAlphaBoolEnabled:bool=True, Translation="0,0,0", Rotation="-5,-1,0", Scale="0.86,0.88,0.88") -> (Pose) {
    alphaInputType=Bool
    rotationMode=BMM_Additive
    rotationSpace=BCS_BoneSpace
    scaleMode=BMM_Additive
    scaleSpace=BCS_BoneSpace
    translationMode=BMM_Additive
    translationSpace=BCS_BoneSpace
    alphaBoolBlend {blendInTime=0.25, blendOutTime=0.25}
    boneToModify {boneName=dyn_katana_blade_l}
    rotation {pitch=-5, yaw=-1}
    scale=<0.86 0.88 0.88>
   }
   n44 = Anim Dynamics - Bone: dyn_ribbon_l(ComponentPose=@n43, Alpha=1, ExternalForce="0,0,0") {
    angularDampingOverride=1
    angularSpringConstant=13000
    bAngularSpring=true
    bOverrideAngularDamping=true
    bUseGravityOverride=true
    gravityScale=0.35
    simulationSpace=RootRelative
    boundBone {boneName=dyn_ribbon_l}
    componentAppliedLinearAccClamp=<3000 3000 3000>
    componentLinearAccScale=<0.6 0.6 0.6>
    componentLinearVelScale=<0.7 0.7 0.7>
    gravityOverride {z=-980}
    physicsBodyDefinitions [{
      boundBone {boneName=dyn_ribbon_l}
      boxExtents=<20 25 25>
      constraintSetup {angularLimitsMax=<0 0 0>, angularLimitsMin=<0 -15 -28>, angularTarget=<1 0 0>}
      localJointOffset=<4 0 0>
     }]
   }
   n45 = Anim Dynamics - Bone: dyn_ribbon_r(ComponentPose=@n44, Alpha=1, ExternalForce="0,0,0") {
    angularDampingOverride=1
    angularSpringConstant=13000
    bAngularSpring=true
    bOverrideAngularDamping=true
    bUseGravityOverride=true
    gravityScale=0.35
    simulationSpace=RootRelative
    boundBone {boneName=dyn_ribbon_r}
    componentAppliedLinearAccClamp=<3000 3000 3000>
    componentLinearAccScale=<0.6 0.6 0.6>
    componentLinearVelScale=<0.7 0.7 0.7>
    gravityOverride {z=-980}
    physicsBodyDefinitions [{
      boundBone {boneName=dyn_ribbon_r}
      boxExtents=<20 25 25>
      constraintSetup {angularLimitsMax=<0 17 28>, angularLimitsMin=<0 0 -28>, angularTarget=<1 0 0>}
      localJointOffset=<4 0 0>
     }]
   }
   n46 = "Transform (Modify) Bone - Bone: base_ribbon_r"(ComponentPose=@n45, Alpha=1, Translation="0,0,0", Rotation="0,25,5", Scale="1,1,1") {rotationMode=BMM_Additive, rotationSpace=BCS_BoneSpace, boneToModify {boneName=base_ribbon_r}, rotation {roll=5, yaw=25}}
   n47 = Get Playing emote()
   n48 = Anim Dynamics - Bone: dyn_pants_in_01_l(ComponentPose=@n46, Alpha=@n47, ExternalForce="0,0,0") {
    angularDampingOverride=1
    angularSpringConstant=12000
    bAngularSpring=true
    bOverrideAngularDamping=true
    bUseGravityOverride=true
    gravityScale=0.35
    simulationSpace=RootRelative
    boundBone {boneName=dyn_pants_in_01_l}
    componentAppliedLinearAccClamp=<3000 3000 3000>
    componentLinearAccScale=<0.4 0.4 0.4>
    componentLinearVelScale=<0.5 0.5 0.5>
    gravityOverride {z=-980}
    physicsBodyDefinitions [{
      boundBone {boneName=dyn_pants_in_01_l}
      boxExtents=<20 30 30>
      constraintSetup {angularLimitsMax=<0 2 0>, angularLimitsMin=<0 0 0>, angularTarget=<1 0 0>}
      localJointOffset=<2.7 0 0>
     }]
    simSpaceSettings {maxAngularAcceleration=50, maxAngularVelocity=5, simSpaceAngularAlpha=1}
   }
   n49 = Anim Dynamics - Bone: dyn_pants_in_02_l!(ComponentPose=@n48, Alpha=@n47, GravityOverride:Vector="0,0,-980", ExternalForce="0,0,0") -> (Pose) {
    angularDampingOverride=1
    angularSpringConstant=2000
    bAngularSpring=true
    bOverrideAngularDamping=true
    bUseGravityOverride=true
    gravityScale=0.35
    simulationSpace=RootRelative
    boundBone {boneName=dyn_pants_in_02_l}
    componentAppliedLinearAccClamp=<5 5 5>
    componentLinearAccScale=<0.1 0.1 0.1>
    componentLinearVelScale=<0.1 0.1 0.1>
    gravityOverride {z=-980}
    physicsBodyDefinitions [{
      boundBone {boneName=dyn_pants_in_02_l}
      boxExtents=<14 14 14>
      constraintSetup {angularLimitsMax=<5 20 18>, angularLimitsMin=<-5 0 -18>, angularTarget=<1 0 0>}
      localJointOffset=<-2.7 0 0>
     }]
    simSpaceSettings {maxAngularAcceleration=50, maxAngularVelocity=5, simSpaceAngularAlpha=1}
   }
   n50 = Anim Dynamics - Bone: dyn_pants_out_01_l(ComponentPose=@n49, Alpha=@n47, ExternalForce="0,0,0") {
    angularDampingOverride=1
    angularSpringConstant=12000
    bAngularSpring=true
    bOverrideAngularDamping=true
    bUseGravityOverride=true
    gravityScale=0.35
    simulationSpace=RootRelative
    boundBone {boneName=dyn_pants_out_01_l}
    componentAppliedLinearAccClamp=<3000 3000 3000>
    componentLinearAccScale=<0.4 0.4 0.4>
    componentLinearVelScale=<0.5 0.5 0.5>
    gravityOverride {z=-980}
    physicsBodyDefinitions [{
      boundBone {boneName=dyn_pants_out_01_l}
      boxExtents=<20 30 30>
      constraintSetup {angularLimitsMax=<0 1 0>, angularLimitsMin=<0 -3 0>, angularTarget=<1 0 0>}
      localJointOffset=<2.7 0 0>
     }]
    simSpaceSettings {maxAngularAcceleration=50, maxAngularVelocity=5, simSpaceAngularAlpha=1}
   }
   n51 = Anim Dynamics - Bone: dyn_pants_out_02_l!(ComponentPose=@n50, Alpha=@n47, GravityOverride:Vector="0,0,-980", ExternalForce="0,0,0") -> (Pose) {
    angularDampingOverride=1
    angularSpringConstant=2000
    bAngularSpring=true
    bOverrideAngularDamping=true
    bUseGravityOverride=true
    gravityScale=0.35
    simulationSpace=RootRelative
    boundBone {boneName=dyn_pants_out_02_l}
    componentAppliedLinearAccClamp=<5 5 5>
    componentLinearAccScale=<0.1 0.1 0.1>
    componentLinearVelScale=<0.1 0.1 0.1>
    gravityOverride {z=-980}
    physicsBodyDefinitions [{
      boundBone {boneName=dyn_pants_out_02_l}
      boxExtents=<14 14 14>
      constraintSetup {angularLimitsMax=<5 15 20>, angularLimitsMin=<-5 -20 -20>, angularTarget=<1 0 0>}
      localJointOffset=<-2.7 0 0>
     }]
    simSpaceSettings {maxAngularAcceleration=50, maxAngularVelocity=5, simSpaceAngularAlpha=1}
   }
   n52 = Anim Dynamics - Bone: dyn_pants_fr_01_l(ComponentPose=@n51, Alpha=@n38, ExternalForce="0,0,0") {
    angularDampingOverride=1
    angularSpringConstant=12000
    bAngularSpring=true
    bOverrideAngularDamping=true
    bUseGravityOverride=true
    gravityScale=0.35
    simulationSpace=RootRelative
    boundBone {boneName=dyn_pants_fr_01_l}
    componentAppliedLinearAccClamp=<3000 3000 3000>
    componentLinearAccScale=<0.4 0.4 0.4>
    componentLinearVelScale=<0.5 0.5 0.5>
    gravityOverride {z=-980}
    physicsBodyDefinitions [{
      boundBone {boneName=dyn_pants_fr_01_l}
      boxExtents=<20 30 30>
      constraintSetup {angularLimitsMax=<0 1 1>, angularLimitsMin=<0 -1 -10>, angularTarget=<1 0 0>}
      localJointOffset=<2.7 0 0>
     }]
    simSpaceSettings {maxAngularAcceleration=50, maxAngularVelocity=5, simSpaceAngularAlpha=1}
   }
   n53 = Anim Dynamics - Bone: dyn_pants_fr_02_l!(ComponentPose=@n52, Alpha=@n38, GravityOverride:Vector="0,0,-980", ExternalForce="0,0,0") -> (Pose) {
    angularDampingOverride=1
    angularSpringConstant=2000
    bAngularSpring=true
    bOverrideAngularDamping=true
    bUseGravityOverride=true
    gravityScale=0.35
    simulationSpace=RootRelative
    boundBone {boneName=dyn_pants_fr_02_l}
    componentAppliedLinearAccClamp=<5 5 5>
    componentLinearAccScale=<0.1 0.1 0.1>
    componentLinearVelScale=<0.1 0.1 0.1>
    gravityOverride {z=-980}
    physicsBodyDefinitions [{
      boundBone {boneName=dyn_pants_fr_02_l}
      boxExtents=<14 14 14>
      constraintSetup {angularLimitsMax=<5 20 10>, angularLimitsMin=<-5 -20 -20>, angularTarget=<1 0 0>}
      localJointOffset=<-2.7 0 0>
     }]
    simSpaceSettings {maxAngularAcceleration=50, maxAngularVelocity=5, simSpaceAngularAlpha=1}
   }
   n54 = Anim Dynamics - Bone: dyn_pants_bk_01_l(ComponentPose=@n53, Alpha=@n38, ExternalForce="0,0,0") {
    angularDampingOverride=1
    angularSpringConstant=12000
    bAngularSpring=true
    bOverrideAngularDamping=true
    bUseGravityOverride=true
    gravityScale=0.35
    simulationSpace=RootRelative
    boundBone {boneName=dyn_pants_bk_01_l}
    componentAppliedLinearAccClamp=<3000 3000 3000>
    componentLinearAccScale=<0.4 0.4 0.4>
    componentLinearVelScale=<0.5 0.5 0.5>
    gravityOverride {z=-980}
    physicsBodyDefinitions [{
      boundBone {boneName=dyn_pants_bk_01_l}
      boxExtents=<20 30 30>
      constraintSetup {angularLimitsMax=<0 1 4>, angularLimitsMin=<0 -1 0>, angularTarget=<1 0 0>}
      localJointOffset=<2.7 0 0>
     }]
    simSpaceSettings {maxAngularAcceleration=50, maxAngularVelocity=5, simSpaceAngularAlpha=1}
   }
   n55 = Anim Dynamics - Bone: dyn_pants_bk_02_l!(ComponentPose=@n54, Alpha=@n38, GravityOverride:Vector="0,0,-980", ExternalForce="0,0,0") -> (Pose) {
    angularDampingOverride=1
    angularSpringConstant=2000
    bAngularSpring=true
    bOverrideAngularDamping=true
    bUseGravityOverride=true
    gravityScale=0.35
    simulationSpace=RootRelative
    boundBone {boneName=dyn_pants_bk_02_l}
    componentAppliedLinearAccClamp=<5 5 5>
    componentLinearAccScale=<0.1 0.1 0.1>
    componentLinearVelScale=<0.1 0.1 0.1>
    gravityOverride {z=-980}
    physicsBodyDefinitions [{
      boundBone {boneName=dyn_pants_bk_02_l}
      boxExtents=<14 14 14>
      constraintSetup {angularLimitsMax=<5 10 20>, angularLimitsMin=<-5 -20 0>, angularTarget=<1 0 0>}
      localJointOffset=<-2.7 0 0>
     }]
    simSpaceSettings {maxAngularAcceleration=50, maxAngularVelocity=5, simSpaceAngularAlpha=1}
   }
   n56 = Anim Dynamics - Bone: dyn_pants_in_01_r(ComponentPose=@n55, Alpha=@n37, ExternalForce="0,0,0") {
    angularDampingOverride=1
    angularSpringConstant=12000
    bAngularSpring=true
    bOverrideAngularDamping=true
    bUseGravityOverride=true
    gravityScale=0.35
    simulationSpace=RootRelative
    boundBone {boneName=dyn_pants_in_01_r}
    componentAppliedLinearAccClamp=<1000 1000 1000>
    componentLinearAccScale=<0.4 0.4 0.4>
    componentLinearVelScale=<0.4 0.4 0.4>
    gravityOverride {z=-980}
    physicsBodyDefinitions [{
      boundBone {boneName=dyn_pants_in_01_r}
      boxExtents=<20 30 30>
      constraintSetup {angularLimitsMax=<0 5 0>, angularLimitsMin=<0 0 0>, angularTarget=<1 0 0>}
      localJointOffset=<3 0 0>
     }]
    simSpaceSettings {maxAngularAcceleration=50, maxAngularVelocity=5, simSpaceAngularAlpha=1}
   }
   n57 = Anim Dynamics - Bone: dyn_pants_in_02_r!(ComponentPose=@n56, Alpha=@n37, GravityOverride:Vector="0,0,-980", ExternalForce="0,0,0") -> (Pose) {
    angularDampingOverride=1
    angularSpringConstant=2000
    bAngularSpring=true
    bOverrideAngularDamping=true
    bUseGravityOverride=true
    gravityScale=0.35
    simulationSpace=RootRelative
    boundBone {boneName=dyn_pants_in_02_r}
    componentAppliedLinearAccClamp=<3 3 3>
    componentLinearAccScale=<0.3 0.3 0.3>
    componentLinearVelScale=<0.3 0.3 0.3>
    gravityOverride {z=-980}
    physicsBodyDefinitions [{
      boundBone {boneName=dyn_pants_in_02_r}
      boxExtents=<14 14 14>
      constraintSetup {angularLimitsMax=<5 0 18>, angularLimitsMin=<-5 -20 -18>, angularTarget=<1 0 0>}
      localJointOffset=<-3.5 0 0>
     }]
    simSpaceSettings {maxAngularAcceleration=50, maxAngularVelocity=5, simSpaceAngularAlpha=1}
   }
   n58 = Anim Dynamics - Bone: dyn_pants_out_01_r(ComponentPose=@n57, Alpha=@n37, ExternalForce="0,0,0") {
    angularDampingOverride=1
    angularSpringConstant=12000
    bAngularSpring=true
    bOverrideAngularDamping=true
    bUseGravityOverride=true
    gravityScale=0.35
    simulationSpace=RootRelative
    boundBone {boneName=dyn_pants_out_01_r}
    componentAppliedLinearAccClamp=<1000 1000 1000>
    componentLinearAccScale=<0.4 0.4 0.4>
    componentLinearVelScale=<0.4 0.4 0.4>
    gravityOverride {z=-980}
    physicsBodyDefinitions [{
      boundBone {boneName=dyn_pants_out_01_r}
      boxExtents=<20 30 30>
      constraintSetup {angularLimitsMax=<0 1 1>, angularLimitsMin=<0 -2 -1>, angularTarget=<1 0 0>}
      localJointOffset=<3 0 0>
     }]
    simSpaceSettings {maxAngularAcceleration=50, maxAngularVelocity=5, simSpaceAngularAlpha=1}
   }
   n59 = Anim Dynamics - Bone: dyn_pants_out_02_r!(ComponentPose=@n58, Alpha=@n37, GravityOverride:Vector="0,0,-980", ExternalForce="0,0,0") -> (Pose) {
    angularDampingOverride=1
    angularSpringConstant=2000
    bAngularSpring=true
    bOverrideAngularDamping=true
    bUseGravityOverride=true
    gravityScale=0.35
    simulationSpace=RootRelative
    boundBone {boneName=dyn_pants_out_02_r}
    componentAppliedLinearAccClamp=<3 3 3>
    componentLinearAccScale=<0.3 0.3 0.3>
    componentLinearVelScale=<0.3 0.3 0.3>
    gravityOverride {z=-980}
    physicsBodyDefinitions [{
      boundBone {boneName=dyn_pants_out_02_r}
      boxExtents=<14 14 14>
      constraintSetup {angularLimitsMax=<5 25 20>, angularLimitsMin=<-5 -2 -20>, angularTarget=<1 0 0>}
      localJointOffset=<-3.5 0 0>
     }]
    simSpaceSettings {maxAngularAcceleration=50, maxAngularVelocity=5, simSpaceAngularAlpha=1}
   }
   n60 = Anim Dynamics - Bone: dyn_pants_fr_01_r(ComponentPose=@n59, Alpha=@n36, ExternalForce="0,0,0") {
    angularDampingOverride=1
    angularSpringConstant=12000
    bAngularSpring=true
    bOverrideAngularDamping=true
    bUseGravityOverride=true
    gravityScale=0.35
    simulationSpace=RootRelative
    boundBone {boneName=dyn_pants_fr_01_r}
    componentAppliedLinearAccClamp=<1000 1000 1000>
    componentLinearAccScale=<0.4 0.4 0.4>
    componentLinearVelScale=<0.4 0.4 0.4>
    gravityOverride {z=-980}
    physicsBodyDefinitions [{
      boundBone {boneName=dyn_pants_fr_01_r}
      boxExtents=<20 30 30>
      constraintSetup {angularLimitsMax=<0 4 2>, angularLimitsMin=<0 -4 -10>, angularTarget=<1 0 0>}
      localJointOffset=<3 0 0>
     }]
    simSpaceSettings {maxAngularAcceleration=50, maxAngularVelocity=5, simSpaceAngularAlpha=1}
   }
   n61 = Anim Dynamics - Bone: dyn_pants_fr_02_r!(ComponentPose=@n60, Alpha=@n36, GravityOverride:Vector="0,0,-980", ExternalForce="0,0,0") -> (Pose) {
    angularDampingOverride=1
    angularSpringConstant=2000
    bAngularSpring=true
    bOverrideAngularDamping=true
    bUseGravityOverride=true
    gravityScale=0.35
    simulationSpace=RootRelative
    boundBone {boneName=dyn_pants_fr_02_r}
    componentAppliedLinearAccClamp=<3 3 3>
    componentLinearAccScale=<0.3 0.3 0.3>
    componentLinearVelScale=<0.3 0.3 0.3>
    gravityOverride {z=-980}
    physicsBodyDefinitions [{
      boundBone {boneName=dyn_pants_fr_02_r}
      boxExtents=<14 14 14>
      constraintSetup {angularLimitsMax=<5 20 2>, angularLimitsMin=<-5 -20 -22>, angularTarget=<1 0 0>}
      localJointOffset=<-3.5 0 0>
     }]
    simSpaceSettings {maxAngularAcceleration=50, maxAngularVelocity=5, simSpaceAngularAlpha=1}
   }
   n62 = Anim Dynamics - Bone: dyn_pants_bk_01_r(ComponentPose=@n61, Alpha=@n36, ExternalForce="0,0,0") {
    angularDampingOverride=1
    angularSpringConstant=12000
    bAngularSpring=true
    bOverrideAngularDamping=true
    bUseGravityOverride=true
    gravityScale=0.35
    simulationSpace=RootRelative
    boundBone {boneName=dyn_pants_bk_01_r}
    componentAppliedLinearAccClamp=<1000 1000 1000>
    componentLinearAccScale=<0.4 0.4 0.4>
    componentLinearVelScale=<0.4 0.4 0.4>
    gravityOverride {z=-980}
    physicsBodyDefinitions [{
      boundBone {boneName=dyn_pants_bk_01_r}
      boxExtents=<20 30 30>
      constraintSetup {angularLimitsMax=<0 1 6>, angularLimitsMin=<0 0 0>, angularTarget=<1 0 0>}
      localJointOffset=<3 0 0>
     }]
    simSpaceSettings {maxAngularAcceleration=50, maxAngularVelocity=5, simSpaceAngularAlpha=1}
   }
   n63 = Anim Dynamics - Bone: dyn_pants_bk_02_r!(ComponentPose=@n62, Alpha=@n36, GravityOverride:Vector="0,0,-980", ExternalForce="0,0,0") -> (Pose) {
    angularDampingOverride=1
    angularSpringConstant=2000
    bAngularSpring=true
    bOverrideAngularDamping=true
    bUseGravityOverride=true
    gravityScale=0.35
    simulationSpace=RootRelative
    boundBone {boneName=dyn_pants_bk_02_r}
    componentAppliedLinearAccClamp=<3 3 3>
    componentLinearAccScale=<0.3 0.3 0.3>
    componentLinearVelScale=<0.3 0.3 0.3>
    gravityOverride {z=-980}
    physicsBodyDefinitions [{
      boundBone {boneName=dyn_pants_bk_02_r}
      boxExtents=<14 14 14>
      constraintSetup {angularLimitsMax=<5 20 23>, angularLimitsMin=<-5 -20 0>, angularTarget=<1 0 0>}
      localJointOffset=<-3.5 0 0>
     }]
    simSpaceSettings {maxAngularAcceleration=50, maxAngularVelocity=5, simSpaceAngularAlpha=1}
   }
   n64 = "Transform (Modify) Bone - Bone: base_pants_bk_01_r"!(ComponentPose=@n63, bAlphaBoolEnabled:bool=True, Translation="0,0,0", Rotation="-3,5,0", Scale="1,1,1") -> (Pose) {
    alphaInputType=Bool
    rotationMode=BMM_Additive
    rotationSpace=BCS_BoneSpace
    alphaBoolBlend {blendInTime=0.25, blendOutTime=0.25}
    boneToModify {boneName=base_pants_bk_01_r}
    rotation {pitch=-3, yaw=5}
   }
   n65 = "Transform (Modify) Bone - Bone: base_pants_in_01_r"!(ComponentPose=@n64, bAlphaBoolEnabled:bool=True, Translation="0,0,0", Rotation="3.5,4.3,0", Scale="1,1,1") -> (Pose) {
    alphaInputType=Bool
    rotationMode=BMM_Additive
    rotationSpace=BCS_BoneSpace
    alphaBoolBlend {blendInTime=0.25, blendOutTime=0.25}
    boneToModify {boneName=base_pants_in_01_r}
    rotation {pitch=3.5, yaw=4.3}
   }
   n66 = "Transform (Modify) Bone - Bone: dyn_pants_m"!(ComponentPose=@n65, bAlphaBoolEnabled:bool=True, Translation="4,0,0", Rotation="0,0,0", Scale="1,1,1") -> (Pose) {
    alphaInputType=Bool
    translationMode=BMM_Additive
    translationSpace=BCS_BoneSpace
    alphaBoolBlend {blendInTime=0.3, blendOutTime=0.3}
    boneToModify {boneName=dyn_pants_m}
    translation {x=4}
   }
   n67 = "Constraint - base_pants_m (PP)"!(ComponentPose=@n66, Alpha=1, ConstraintWeights_0=0.5, ConstraintWeights_1:float=0.5) -> (Pose) {
    boneToModify {boneName=base_pants_m}
    constraintSetup [{targetBone {boneName=thigh_l}}, {targetBone {boneName=thigh_r}}]
    constraintWeights [1,1]
   }
   Save cached pose 'Non sleeve dynamics'(Pose=Component To Local(ComponentPose=@n67)) {cacheName=Non sleeve dynamics}
  ]}
 EventGraph {nodes [
   Event BlueprintUpdateAnimation()
   > n1 = Sequence!(execute) -> (then_0, then_1, then_2:Exec, then_3:Exec, then_4:Exec, then_5:Exec, then_6:Exec, then_7:Exec)
   n2 = "Gravity Override State Function (Parameter Asset) (Nativized)"(
    execute=@n1.then_1
    CharacterPartAnimInstance=Self-Reference()
    GameTimeInSeconds=GetGameTimeInSeconds()
    Parameters=&/F_MED_MajorSpeech_Parameters
   )
   > Set GravityOverride(GravityOverride=@n2)
   n3 = Get bIsRightArmHidden() -> (bIsRightArmHidden:bool)
   Set IsWearingGauntlet_alpha_Right(execute=@n1.then_4, IsWearingGauntlet_alpha_Right:double="To Float (Boolean)"(InBool=NOT Boolean(A=@n3))) -> (Output_Get:double)
   n4 = Branch(execute=@n1.then_2, Condition=Get bIsPlayingEmote())
   > Set Playing emote(Playing emote:double=0.45) -> (Output_Get:double)
   n5 = Branch(execute=@n4.else, Condition=OR Boolean(A=Get bIsParachuteOpened(), B=Get bIsSkyDiving()))
   > Set Playing emote(Playing emote=0.8)
   Set Playing emote(execute=@n5.else, Playing emote=1)
   Set JumpFall(execute=@n1.then_0, JumpFall:bool=OR Boolean(A=Get bIsJumping(), B=Get bIsFalling())) -> (Output_Get:bool)
   > Set Moving on crouch(Moving on crouch:bool=AND Boolean(A=Get bIsCrouching(), B=Get bIsAccelerating2D())) -> (Output_Get:bool)
   > Branch(Condition=Get bIsInFrontEnd())
   > Set Custom Idle Is Playing Float(Custom Idle Is Playing Float:double=GetCurveValue(self=GetOwnerFrontEndAnimInstance(), CurveName=Frontend_Idle_Bypass)) -> (Output_Get:double)
   n6 = Get bIsLeftArmHidden() -> (bIsLeftArmHidden:bool)
   Set IsWearingGauntlet_alpha_Left(execute=@n1.then_3, IsWearingGauntlet_alpha_Left:double="To Float (Boolean)"(InBool=NOT Boolean(A=@n6))) -> (Output_Get:double)
   n7 = Get bIsTargeting()
   n8 = Get bIsCrouching()
   Set CrouchTarget(CrouchTarget:bool=AND Boolean(A=@n7, B=@n8)) -> (Output_Get:bool)
   Set Target(Target:bool=AND Boolean(A=@n7, B=NOT Boolean(A=@n8))) -> (Output_Get:bool)
  ]}
]
variables [
 CrouchTarget {type=boolean}
 Custom Idle Is Playing Float {type=number}
 FrontEnd_alpha {type=number}
 FrontEndRot_alpha {type=number}
 GravityOverride {type=Vector}
 IsWearingGauntlet_alpha_Left {type=number}
 IsWearingGauntlet_alpha_Right {type=number}
 JumpFall {type=boolean}
 katana_rotate {type=Rotator}
 Moving on crouch {type=boolean}
 Playing emote {type=number}
 Sleeve_rotator {type=Rotator}
 SpineRot {type=number}
 Target {type=boolean}
]
