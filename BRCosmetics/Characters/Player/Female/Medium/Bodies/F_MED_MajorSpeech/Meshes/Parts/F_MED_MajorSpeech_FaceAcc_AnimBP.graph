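# Anim Blueprint for Major Speech's FaceAcc part, adding secondary hair motion over the pose copied from the attached parent mesh.
# Center-hair AnimDynamics feeds a cached pose, then bIsSkyDiving blends between two spring-dynamics chains, the skydive one pre-posing hair base bones.
# EventGraph remaps head-to-neck yaw into hair offsets, of which the normal branch applies back-hair rotation and translation.
# Several node chains and the computed GravityOverride never reach the output.
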AnimBlueprint /BRCosmetics/Characters/Player/Female/Medium/Bodies/F_MED_MajorSpeech/Meshes/Parts/F_MED_MajorSpeech_FaceAcc_AnimBP
target_skeleton=&/F_MED_MajorSpeech_FaceAcc_Skeleton
graphs [
 AnimGraph {nodes [
   n1 = Anim Dynamics - Bone: dyn_hair_01_l(Alpha=1, ExternalForce="0,0,0") {
    angularDampingOverride=1
    angularSpringConstant=5000
    bAngularSpring=true
    bOverrideAngularDamping=true
    bUseGravityOverride=true
    gravityScale=0.35
    simulationSpace=BoneRelative
    boundBone {boneName=dyn_hair_01_l}
    componentAppliedLinearAccClamp=<2000 2000 2000>
    componentLinearAccScale=<0.3 0.3 0.3>
    componentLinearVelScale=<0.6 0.6 0.6>
    gravityOverride {z=-970}
    physicsBodyDefinitions [{
      boundBone {boneName=dyn_hair_01_l}
      boxExtents=<20 24 24>
      constraintSetup {angularLimitsMax=<0 0 15>, angularLimitsMin=<0 -15 -15>, angularTarget=<1 0 0>}
      localJointOffset=<3.8 0 0>
     }]
    relativeSpaceBone {boneName=spine_05}
    simSpaceSettings {maxAngularAcceleration=100, maxAngularVelocity=10, simSpaceAngularAlpha=1}
   }
   n2 = Anim Dynamics - Bone: dyn_hair_01_r(ComponentPose=@n1, Alpha=1, ExternalForce="0,0,0") {
    angularDampingOverride=1
    angularSpringConstant=5000
    bAngularSpring=true
    bOverrideAngularDamping=true
    bUseGravityOverride=true
    gravityScale=0.35
    simulationSpace=BoneRelative
    boundBone {boneName=dyn_hair_01_r}
    componentAppliedLinearAccClamp=<2000 2000 2000>
    componentLinearAccScale=<0.4 0.3 0.4>
    componentLinearVelScale=<0.6 0.6 0.6>
    gravityOverride {z=-970}
    physicsBodyDefinitions [{
      boundBone {boneName=dyn_hair_01_r}
      boxExtents=<20 23 23>
      constraintSetup {angularLimitsMax=<0 22 30>, angularLimitsMin=<0 0 -30>, angularTarget=<1 0 0>}
      localJointOffset=<3.8 0 0>
     }]
    relativeSpaceBone {boneName=spine_05}
    simSpaceSettings {maxAngularAcceleration=100, maxAngularVelocity=10, simSpaceAngularAlpha=1}
   }
   n3 = Anim Dynamics - Bone: dyn_hair_01_ft_l(ComponentPose=@n2, Alpha=1, ExternalForce="0,0,0") {
    angularDampingOverride=1
    angularSpringConstant=5000
    bAngularSpring=true
    bOverrideAngularDamping=true
    bUseGravityOverride=true
    gravityScale=0.35
    simulationSpace=BoneRelative
    boundBone {boneName=dyn_hair_01_ft_l}
    componentAppliedLinearAccClamp=<2000 2000 2000>
    componentLinearAccScale=<0.5 0.5 0.5>
    componentLinearVelScale=<0.6 0.6 0.6>
    gravityOverride {z=-970}
    physicsBodyDefinitions [{
      boundBone {boneName=dyn_hair_01_ft_l}
      boxExtents=<20 25 25>
      constraintSetup {angularLimitsMax=<0 0 20>, angularLimitsMin=<0 -35 -20>, angularTarget=<1 0 0>}
      localJointOffset=<4.2 0 0>
     }]
    relativeSpaceBone {boneName=spine_05}
    simSpaceSettings {maxAngularAcceleration=100, maxAngularVelocity=10, simSpaceAngularAlpha=1}
   }
   n4 = Anim Dynamics - Bone: dyn_hair_01_ft_r(ComponentPose=@n3, Alpha=1, ExternalForce="0,0,0") {
    angularDampingOverride=1
    angularSpringConstant=5000
    bAngularSpring=true
    bOverrideAngularDamping=true
    bUseGravityOverride=true
    gravityScale=0.35
    simulationSpace=BoneRelative
    boundBone {boneName=dyn_hair_01_ft_r}
    componentAppliedLinearAccClamp=<2000 2000 2000>
    componentLinearAccScale=<0.5 0.5 0.5>
    componentLinearVelScale=<0.6 0.6 0.6>
    gravityOverride {z=-970}
    physicsBodyDefinitions [{
      boundBone {boneName=dyn_hair_01_ft_r}
      boxExtents=<20 25 25>
      constraintSetup {angularLimitsMax=<0 15 25>, angularLimitsMin=<0 0 -25>, angularTarget=<1 0 0>}
      localJointOffset=<4.2 0 0>
     }]
    relativeSpaceBone {boneName=spine_05}
    simSpaceSettings {maxAngularAcceleration=100, maxAngularVelocity=10, simSpaceAngularAlpha=1}
   }
   Anim Dynamics - Bone: dyn_hair_out_01_r(ComponentPose=@n4, Alpha=1, ExternalForce="0,0,0") {
    angularDampingOverride=1
    angularSpringConstant=9000
    bAngularSpring=true
    bOverrideAngularDamping=true
    bUseGravityOverride=true
    gravityScale=0.35
    simulationSpace=BoneRelative
    boundBone {boneName=dyn_hair_out_01_r}
    componentLinearAccScale=<0.4 0.3 0.4>
    componentLinearVelScale=<0.6 0.6 0.6>
    gravityOverride {z=-980}
    physicsBodyDefinitions [{
      boundBone {boneName=dyn_hair_out_01_r}
      boxExtents=<20 27 27>
      constraintSetup {angularLimitsMax=<0 12 65>, angularLimitsMin=<0 -16 -55>, angularTarget=<1 0 0>}
      localJointOffset=<3.9 0 0>
     }]
    relativeSpaceBone {boneName=spine_05}
   }
   n5 = Get hair_sides_rotate() -> (hair_sides_rotate:Rotator)
   n6 = "Transform (Modify) Bone - Bone: base_hair_l"(Alpha=1, Translation="0,0,0", Rotation=@n5, Scale="1,1,1") {
    rotationMode=BMM_Additive
    rotationSpace=BCS_BoneSpace
    alphaBoolBlend {blendInTime=0.2, blendOutTime=0.2}
    boneToModify {boneName=base_hair_l}
   }
   n7 = "Transform (Modify) Bone - Bone: base_hair_r"(ComponentPose=@n6, Alpha=1, Translation="0,0,0", Rotation=@n5, Scale="1,1,1") {
    rotationMode=BMM_Additive
    rotationSpace=BCS_BoneSpace
    alphaBoolBlend {blendInTime=0.2, blendOutTime=0.2}
    boneToModify {boneName=base_hair_r}
   }
   n8 = "Transform (Modify) Bone - Bone: base_hair_ft_l"(ComponentPose=@n7, Alpha=1, Translation="0,0,0", Rotation=@n5, Scale="1,1,1") {
    rotationMode=BMM_Additive
    rotationSpace=BCS_BoneSpace
    translationMode=BMM_Additive
    translationSpace=BCS_BoneSpace
    alphaBoolBlend {blendInTime=0.2, blendOutTime=0.2}
    boneToModify {boneName=base_hair_ft_l}
   }
   "Transform (Modify) Bone - Bone: base_hair_ft_r"(ComponentPose=@n8, Alpha=1, Translation="0,0,0", Rotation=@n5, Scale="1,1,1") {
    rotationMode=BMM_Additive
    rotationSpace=BCS_BoneSpace
    translationMode=BMM_Additive
    translationSpace=BCS_BoneSpace
    alphaBoolBlend {blendInTime=0.2, blendOutTime=0.2}
    boneToModify {boneName=base_hair_ft_r}
   }
   "Transform (Modify) Bone - Bone: base_hair_bk_c"(Alpha=1, Translation="1.5,2,0", Rotation="0,85,0", Scale="1,1,1") {
    rotationMode=BMM_Additive
    rotationSpace=BCS_BoneSpace
    translationMode=BMM_Additive
    translationSpace=BCS_BoneSpace
    alphaBoolBlend {blendInTime=0.2, blendOutTime=0.2}
    boneToModify {boneName=base_hair_bk_c}
    rotation {yaw=85}
    translation=<1.5 2>
   }
   n9 = Anim Dynamics - Bone: dyn_hair_01_bk_c(ComponentPose=Local To Component(LocalPose=Use cached pose 'Input Pose Cache'()), Alpha=1, ExternalForce="0,0,0") {
    angularDampingOverride=1
    angularSpringConstant=9000
    bAngularSpring=true
    bOverrideAngularDamping=true
    bUseGravityOverride=true
    gravityScale=0.35
    simulationSpace=BoneRelative
    boundBone {boneName=dyn_hair_01_bk_c}
    componentAppliedLinearAccClamp=<2000 2000 2000>
    componentLinearAccScale=<0.2 0.2 0.2>
    componentLinearVelScale=<0.4 0.4 0.4>
    gravityOverride {z=-950}
    physicsBodyDefinitions [{
      boundBone {boneName=dyn_hair_01_bk_c}
      boxExtents=<20 23 23>
      constraintSetup {angularLimitsMax=<0 7 0>, angularLimitsMin=<0 -10 0>, angularTarget=<1 0 0>}
      localJointOffset=<3.8 0 0>
     }]
    relativeSpaceBone {boneName=spine_05}
    simSpaceSettings {maxAngularAcceleration=100, maxAngularVelocity=10, simSpaceAngularAlpha=1}
   }
   n10 = Anim Dynamics - Bone: dyn_hair_01_ft_l(ComponentPose=@n9, Alpha=1, ExternalForce="0,0,0") {
    angularDampingOverride=1
    angularSpringConstant=9000
    bAngularSpring=true
    bOverrideAngularDamping=true
    bUseGravityOverride=true
    gravityScale=0.35
    simulationSpace=RootRelative
    boundBone {boneName=dyn_hair_01_ft_l}
    componentAppliedLinearAccClamp=<5000 5000 5000>
    componentLinearAccScale=<1 1 1>
    componentLinearVelScale=<1 1 1>
    gravityOverride {z=-980}
    physicsBodyDefinitions [{
      boundBone {boneName=dyn_hair_01_ft_l}
      boxExtents=<20 25 25>
      constraintSetup {angularLimitsMax=<0 0 50>, angularLimitsMin=<0 -25 -50>, angularTarget=<1 0 0>}
      localJointOffset=<4.2 0 0>
     }]
    relativeSpaceBone {boneName=spine_05}
   }
   n11 = Anim Dynamics - Bone: dyn_hair_01_ft_r(ComponentPose=@n10, Alpha=1, ExternalForce="0,0,0") {
    angularDampingOverride=1
    angularSpringConstant=9000
    bAngularSpring=true
    bOverrideAngularDamping=true
    bUseGravityOverride=true
    gravityScale=0.35
    simulationSpace=RootRelative
    boundBone {boneName=dyn_hair_01_ft_r}
    componentAppliedLinearAccClamp=<5000 5000 5000>
    componentLinearAccScale=<1 1 1>
    componentLinearVelScale=<1 1 1>
    gravityOverride {z=-980}
    physicsBodyDefinitions [{
      boundBone {boneName=dyn_hair_01_ft_r}
      boxExtents=<20 25 25>
      constraintSetup {angularLimitsMax=<0 25 50>, angularLimitsMin=<0 0 -50>, angularTarget=<1 0 0>}
      localJointOffset=<4.2 0 0>
     }]
    relativeSpaceBone {boneName=spine_05}
   }
   n12 = Anim Dynamics - Bone: dyn_hair_out_01_r(ComponentPose=@n11, Alpha=1, ExternalForce="0,0,0") {
    angularDampingOverride=1
    angularSpringConstant=9000
    bAngularSpring=true
    bOverrideAngularDamping=true
    bUseGravityOverride=true
    gravityScale=0.35
    simulationSpace=RootRelative
    boundBone {boneName=dyn_hair_out_01_r}
    componentAppliedLinearAccClamp=<5000 5000 5000>
    componentLinearAccScale=<1 1 1>
    componentLinearVelScale=<1 1 1>
    gravityOverride {z=-980}
    physicsBodyDefinitions [{
      boundBone {boneName=dyn_hair_out_01_r}
      boxExtents=<20 25 25>
      constraintSetup {angularLimitsMax=<0 25 50>, angularLimitsMin=<0 -7 -50>, angularTarget=<1 0 0>}
      localJointOffset=<4.2 0 0>
     }]
    relativeSpaceBone {boneName=spine_05}
   }
   n13 = Anim Dynamics - Bone: dyn_hair_01_l(ComponentPose=@n12, Alpha=1, ExternalForce="0,0,0") {
    angularDampingOverride=1
    angularSpringConstant=9000
    bAngularSpring=true
    bOverrideAngularDamping=true
    bUseGravityOverride=true
    gravityScale=0.35
    simulationSpace=RootRelative
    boundBone {boneName=dyn_hair_01_l}
    componentAppliedLinearAccClamp=<5000 5000 5000>
    componentLinearAccScale=<1 1 1>
    componentLinearVelScale=<1 1 1>
    gravityOverride {z=-980}
    physicsBodyDefinitions [{
      boundBone {boneName=dyn_hair_01_l}
      boxExtents=<20 25 25>
      constraintSetup {angularLimitsMax=<0 -2 65>, angularLimitsMin=<0 -25 -65>, angularTarget=<1 0 0>}
      localJointOffset=<4.2 0 0>
     }]
    relativeSpaceBone {boneName=spine_05}
   }
   n14 = Anim Dynamics - Bone: dyn_hair_01_r(ComponentPose=@n13, Alpha=1, ExternalForce="0,0,0") {
    angularDampingOverride=1
    angularSpringConstant=9000
    bAngularSpring=true
    bOverrideAngularDamping=true
    bUseGravityOverride=true
    gravityScale=0.35
    simulationSpace=RootRelative
    boundBone {boneName=dyn_hair_01_r}
    componentAppliedLinearAccClamp=<5000 5000 5000>
    componentLinearAccScale=<1 1 1>
    componentLinearVelScale=<1 1 1>
    gravityOverride {z=-980}
    physicsBodyDefinitions [{
      boundBone {boneName=dyn_hair_01_r}
      boxExtents=<20 25 25>
      constraintSetup {angularLimitsMax=<0 25 65>, angularLimitsMin=<0 2 -65>, angularTarget=<1 0 0>}
      localJointOffset=<4.2 0 0>
     }]
    relativeSpaceBone {boneName=spine_05}
   }
   n15 = "Transform (Modify) Bone - Bone: base_hair_bk_c"(
    ComponentPose="Transform (Modify) Bone - Bone: base_hair_ft_r"(
     ComponentPose="Transform (Modify) Bone - Bone: base_hair_ft_l"(ComponentPose=@n14, Alpha=1, Translation="0,0,0", Rotation="9,0,0", Scale="1,1,1") {rotationMode=BMM_Additive, rotationSpace=BCS_BoneSpace, boneToModify {boneName=base_hair_ft_l}, rotation {pitch=9}}
     Alpha=1
     Translation="0,0,0"
     Rotation="-8,0,0"
     Scale="1,1,1"
    ) {rotationMode=BMM_Additive, rotationSpace=BCS_BoneSpace, boneToModify {boneName=base_hair_ft_r}, rotation {pitch=-8}}
    Alpha=1
    Translation=Get hair_translate() -> (hair_translate:Vector)
    Rotation=Get hair_rotate() -> (hair_rotate:Rotator)
    Scale="1,1,1"
   ) {
    rotationMode=BMM_Additive
    rotationSpace=BCS_BoneSpace
    translationMode=BMM_Additive
    translationSpace=BCS_BoneSpace
    alphaBoolBlend {blendInTime=0.2, blendOutTime=0.2}
    boneToModify {boneName=base_hair_bk_c}
   }
   n16 = "Transform (Modify) Bone - Bone: base_hair_bk_c"(
    ComponentPose=Local To Component(LocalPose=Use cached pose 'Input Pose Cache'())
    Alpha=1
    Translation="-0.9,-0.8,0"
    Rotation="0,95,0"
    Scale="1,1,1"
   ) {
    rotationMode=BMM_Additive
    rotationSpace=BCS_BoneSpace
    translationMode=BMM_Additive
    translationSpace=BCS_BoneSpace
    alphaBoolBlend {blendInTime=0.2, blendOutTime=0.2}
    boneToModify {boneName=base_hair_bk_c}
    rotation {yaw=95}
    translation=<-0.9 -0.8>
   }
   n17 = "Transform (Modify) Bone - Bone: base_hair_l"(ComponentPose=@n16, Alpha=1, Translation="0.5,0,0", Rotation="12,35,0", Scale="1,1,1") {
    rotationMode=BMM_Additive
    rotationSpace=BCS_BoneSpace
    alphaBoolBlend {blendInTime=0.2, blendOutTime=0.2}
    boneToModify {boneName=base_hair_l}
    rotation {pitch=12, yaw=35}
    translation {x=0.5}
   }
   n18 = "Transform (Modify) Bone - Bone: base_hair_r"(ComponentPose=@n17, Alpha=1, Translation="0.5,0,0", Rotation="-20,35,0", Scale="1,1,1") {
    rotationMode=BMM_Additive
    rotationSpace=BCS_BoneSpace
    alphaBoolBlend {blendInTime=0.2, blendOutTime=0.2}
    boneToModify {boneName=base_hair_r}
    rotation {pitch=-20, yaw=35}
    translation {x=0.5}
   }
   n19 = Anim Dynamics - Bone: dyn_hair_01_l!(ComponentPose=@n18, Alpha=1, GravityOverride:Vector="0,0,-980", ExternalForce="0,0,0") -> (Pose) {
    angularDampingOverride=1
    angularSpringConstant=9000
    bAngularSpring=true
    bOverrideAngularDamping=true
    bUseGravityOverride=true
    gravityScale=0.35
    simulationSpace=RootRelative
    boundBone {boneName=dyn_hair_01_l}
    componentAppliedLinearAccClamp=<2000 2000 2000>
    componentLinearAccScale=<0.3 0.3 0.3>
    componentLinearVelScale=<0.5 0.5 0.5>
    gravityOverride {z=-980}
    physicsBodyDefinitions [{
      boundBone {boneName=dyn_hair_01_l}
      boxExtents=<20 30 30>
      constraintSetup {angularLimitsMax=<0 0 10>, angularLimitsMin=<0 -18 -7>, angularTarget=<1 0 0>}
      localJointOffset=<3.2 0 0>
     }]
   }
   n20 = Anim Dynamics - Bone: dyn_hair_01_r!(ComponentPose=@n19, Alpha=1, GravityOverride:Vector="0,0,-980", ExternalForce="0,0,0") -> (Pose) {
    angularDampingOverride=1
    angularSpringConstant=11000
    bAngularSpring=true
    bOverrideAngularDamping=true
    bUseGravityOverride=true
    gravityScale=0.35
    simulationSpace=RootRelative
    boundBone {boneName=dyn_hair_01_r}
    componentAppliedLinearAccClamp=<2000 2000 2000>
    componentLinearAccScale=<0.3 0.3 0.3>
    componentLinearVelScale=<0.4 0.4 0.4>
    gravityOverride {z=-980}
    physicsBodyDefinitions [{
      boundBone {boneName=dyn_hair_01_r}
      boxExtents=<20 30 30>
      constraintSetup {angularLimitsMax=<0 15 15>, angularLimitsMin=<0 0 -10>, angularTarget=<1 0 0>}
      localJointOffset=<3.3 0 0>
     }]
   }
   n21 = Anim Dynamics - Bone: dyn_hair_01_bk_c!(ComponentPose=@n20, Alpha=1, GravityOverride:Vector="0,0,-980", ExternalForce="0,0,0") -> (Pose) {
    angularDampingOverride=1
    angularSpringConstant=9000
    bAngularSpring=true
    bOverrideAngularDamping=true
    bUseGravityOverride=true
    gravityScale=0.35
    simulationSpace=RootRelative
    boundBone {boneName=dyn_hair_01_bk_c}
    componentLinearAccScale=<0.3 0.3 0.3>
    componentLinearVelScale=<0.5 0.5 0.5>
    gravityOverride {z=-980}
    physicsBodyDefinitions [{
      boundBone {boneName=dyn_hair_01_bk_c}
      boxExtents=<20 26 26>
      constraintSetup {angularLimitsMax=<0 10 8>, angularLimitsMin=<0 -10 -8>, angularTarget=<1 0 0>}
      localJointOffset=<3.3 0 0>
     }]
   }
   n22 = Anim Dynamics - Bone: dyn_hair_01_ft_l!(ComponentPose=@n21, Alpha=1, GravityOverride:Vector="0,0,-950", ExternalForce="0,0,0") -> (Pose) {
    angularDampingOverride=1
    angularSpringConstant=9000
    bAngularSpring=true
    bOverrideAngularDamping=true
    bUseGravityOverride=true
    gravityScale=0.35
    simulationSpace=RootRelative
    boundBone {boneName=dyn_hair_01_ft_l}
    componentAppliedLinearAccClamp=<2000 2000 2000>
    componentLinearAccScale=<0.5 0.5 0.5>
    componentLinearVelScale=<0.6 0.6 0.6>
    gravityOverride {z=-950}
    physicsBodyDefinitions [{
      boundBone {boneName=dyn_hair_01_ft_l}
      boxExtents=<20 25 25>
      constraintSetup {angularLimitsMax=<0 0 15>, angularLimitsMin=<0 -20 -15>, angularTarget=<1 0 0>}
      localJointOffset=<3.7 0 0>
     }]
   }
   n23 = Anim Dynamics - Bone: dyn_hair_01_ft_r!(ComponentPose=@n22, Alpha=1, GravityOverride:Vector="0,0,-950", ExternalForce="0,0,0") -> (Pose) {
    angularDampingOverride=1
    angularSpringConstant=9000
    bAngularSpring=true
    bOverrideAngularDamping=true
    bUseGravityOverride=true
    gravityScale=0.35
    simulationSpace=RootRelative
    boundBone {boneName=dyn_hair_01_ft_r}
    componentAppliedLinearAccClamp=<2000 2000 2000>
    componentLinearAccScale=<0.5 0.5 0.5>
    componentLinearVelScale=<0.6 0.6 0.6>
    gravityOverride {z=-950}
    physicsBodyDefinitions [{
      boundBone {boneName=dyn_hair_01_ft_r}
      boxExtents=<20 25 25>
      constraintSetup {angularLimitsMax=<0 20 15>, angularLimitsMin=<0 0 -15>, angularTarget=<1 0 0>}
      localJointOffset=<3.7 0 0>
     }]
   }
   n24 = Anim Dynamics - Bone: dyn_hair_out_01_r!(ComponentPose=@n23, Alpha=1, GravityOverride:Vector="0,0,-980", ExternalForce="0,0,0") -> (Pose) {
    angularDampingOverride=1
    angularSpringConstant=10000
    bAngularSpring=true
    bOverrideAngularDamping=true
    bUseGravityOverride=true
    gravityScale=0.35
    simulationSpace=RootRelative
    boundBone {boneName=dyn_hair_out_01_r}
    componentLinearAccScale=<0.3 0.3 0.3>
    componentLinearVelScale=<0.3 0.3 0.3>
    gravityOverride {z=-980}
    physicsBodyDefinitions [{
      boundBone {boneName=dyn_hair_out_01_r}
      boxExtents=<20 27 27>
      constraintSetup {angularLimitsMax=<0 18 15>, angularLimitsMin=<0 0 -15>, angularTarget=<1 0 0>}
      localJointOffset=<3.9 0 0>
     }]
   }
   Output Pose(Result=Blend Poses by bool(
     BlendPose_0=Component To Local(ComponentPose=@n24)
     BlendPose_1=Component To Local(ComponentPose=@n15)
     BlendTime_0=0.25
     BlendTime_1=0.25
     bActiveValue=Get bIsSkyDiving()
    )) {layerGroup=None}
   n25 = "Transform (Modify) Bone - Bone: base_hair_c"(
    ComponentPose=Local To Component(LocalPose=Copy Pose From Mesh() {bUseAttachedParent=true})
    Alpha=1
    Translation="0,0,0"
    Rotation="0,4,0"
    Scale="1,1,1"
   ) {
    rotationMode=BMM_Additive
    rotationSpace=BCS_BoneSpace
    alphaBoolBlend {blendInTime=0.2, blendOutTime=0.2}
    boneToModify {boneName=base_hair_c}
    rotation {yaw=4}
   }
   n26 = Anim Dynamics - Bone: dyn_hair_01_c(ComponentPose=@n25, Alpha=1, ExternalForce="0,0,0") {
    angularDampingOverride=1
    angularSpringConstant=9000
    bAngularSpring=true
    bOverrideAngularDamping=true
    bUseGravityOverride=true
    gravityScale=0.35
    simulationSpace=RootRelative
    boundBone {boneName=dyn_hair_01_c}
    componentAppliedLinearAccClamp=<5000 5000 5000>
    componentLinearAccScale=<0.6 0.6 0.6>
    componentLinearVelScale=<0.6 0.6 0.6>
    gravityOverride {z=-980}
    physicsBodyDefinitions [{
      boundBone {boneName=dyn_hair_01_c}
      boxExtents=<20 26 20>
      constraintSetup {angularLimitsMax=<0 4 75>, angularLimitsMin=<0 0 0>, angularTarget=<1 0 0>}
      localJointOffset=<4.5 0 0>
     }]
   }
   n27 = Anim Dynamics - Bone: dyn_hair_02_c(ComponentPose=@n26, Alpha=1, ExternalForce="0,0,0") {
    angularDampingOverride=1
    angularSpringConstant=9000
    bAngularSpring=true
    bOverrideAngularDamping=true
    bUseGravityOverride=true
    gravityScale=0.35
    simulationSpace=RootRelative
    boundBone {boneName=dyn_hair_02_c}
    componentAppliedLinearAccClamp=<5000 5000 5000>
    componentLinearAccScale=<0.6 0.6 0.6>
    componentLinearVelScale=<0.6 0.6 0.6>
    gravityOverride {z=-980}
    physicsBodyDefinitions [{
      boundBone {boneName=dyn_hair_02_c}
      boxExtents=<20 26 19>
      constraintSetup {angularLimitsMax=<0 7 75>, angularLimitsMin=<0 0 0>, angularTarget=<1 0 0>}
      localJointOffset=<4.5 0 0>
     }]
   }
   n28 = Anim Dynamics - Bone: dyn_hair_03_c(ComponentPose=@n27, Alpha=1, ExternalForce="0,0,0") {
    angularDampingOverride=1
    angularSpringConstant=9000
    bAngularSpring=true
    bOverrideAngularDamping=true
    bUseGravityOverride=true
    gravityScale=0.35
    simulationSpace=RootRelative
    boundBone {boneName=dyn_hair_03_c}
    componentAppliedLinearAccClamp=<5000 5000 5000>
    componentLinearAccScale=<0.6 0.6 0.6>
    componentLinearVelScale=<0.6 0.6 0.6>
    gravityOverride {z=-980}
    physicsBodyDefinitions [{
      boundBone {boneName=dyn_hair_03_c}
      boxExtents=<20 26 20>
      constraintSetup {angularLimitsMax=<0 15 75>, angularLimitsMin=<0 0 0>, angularTarget=<1 0 0>}
      localJointOffset=<4.5 0 0>
     }]
   }
   Save cached pose 'Input Pose Cache'(Pose=Component To Local(ComponentPose=@n28)) {cacheName=Input Pose Cache}
   Get hair_sides_translate() -> (hair_sides_translate:Vector)
  ]}
 EventGraph {nodes [
   Event BlueprintUpdateAnimation()
   > n1 = Sequence()
   > n2 = Branch(Condition=Get bIsParachuteOpened())
   > Set NeckAlpha(NeckAlpha:double=0.2) -> (Output_Get:double)
   n3 = GetOwnerSkeletalMeshComponent()
   n4 = BreakRotator(InRot=BreakTransform(InTransform=GetDeltaTransformFromRefPose(self=@n3, BoneName=head, BaseName=neck_01)).Rotation)
   Set NeckAlpha(execute=@n2.else, NeckAlpha=MapRangeClamped(Value=@n4.Yaw, InRangeA=2, InRangeB=110, OutRangeB=5))
   > Set NeckAlpha_skydive(NeckAlpha_skydive:double=MapRangeClamped(Value=@n4.Yaw, InRangeA=0, InRangeB=-55)) -> (Output_Get:double)
   > n5 = Sequence()
   > Set hair_rotate_front(hair_rotate_front:Rotator=MakeRotator(Yaw=MapRangeClamped(Value=@n4.Yaw, InRangeA=-45, InRangeB=90, OutRangeA=1, OutRangeB=-45))) -> (Output_Get:Rotator)
   Set hair_rotate(
    execute=@n5.then_1
    hair_rotate:Rotator=MakeRotator(Yaw=MapRangeClamped(Value=@n4.Yaw, InRangeA=29, InRangeB=-33, OutRangeA=-45, OutRangeB=50))
   ) -> (Output_Get:Rotator)
   > Set hair_translate(hair_translate:Vector=MakeVector(X=MapRangeClamped(Value=@n4.Yaw, InRangeA=23, InRangeB=-31, OutRangeA=1.5, OutRangeB=-1.7), Y=0)) -> (Output_Get:Vector)
   > Set hair_sides_rotate(hair_sides_rotate:Rotator=MakeRotator(Yaw=MapRangeClamped(Value=@n4.Yaw, InRangeA=29, InRangeB=-30, OutRangeA=-30, OutRangeB=55))) -> (Output_Get:Rotator)
   > Set hair_sides_translate(hair_sides_translate:Vector=MakeVector(X=MapRangeClamped(Value=@n4.Yaw, InRangeA=23, InRangeB=-31, OutRangeA=1, OutRangeB=-1), Y=0)) -> (Output_Get:Vector)
   > Set NeckAlpha_front(NeckAlpha_front:double=MapRangeClamped(Value=@n4.Yaw, InRangeA=0, InRangeB=20, OutRangeA=0.05)) -> (Output_Get:double)
   n6 = "Gravity Override State Function (Parameter Asset) (Nativized)"(
    execute=@n1.then_1
    CharacterPartAnimInstance=Self-Reference()
    GameTimeInSeconds=GetGameTimeInSeconds()
    Parameters=&/F_MED_MajorSpeech_FaceAcc_Parameters
   )
   > Set GravityOverride(GravityOverride=@n6)
  ]}
]
variables [
 GravityOverride {type=Vector}
 hair_rotate {type=Rotator}
 hair_rotate_front {type=Rotator}
 hair_sides_rotate {type=Rotator}
 hair_sides_translate {type=Vector}
 hair_translate {type=Vector}
 HeadRot {type=boolean}
 NeckAlpha {type=number}
 NeckAlpha_front {type=number}
 NeckAlpha_skydive {type=number}
]
